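# WBP_FMJamKeyPreview is a small text widget that displays the musical key of the owning player's jam session.
# PreConstruct fills a label array by iterating EMusicKey, giving accidentals both sharp and flat names.
# Construct and Refresh query the session's music manager and, on success, show the label indexed by the key's byte value.
# OverrideKeyEntry displays a supplied key instead, without querying the session.
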WidgetBlueprint /FMJamSystem/Widgets/GlobalControls/WBP_FMJamKeyPreview
graphs [
 EventGraph {nodes [
   n1 = Refresh() {is=title K2Node_CustomEvent}
   n2 = GetJamSessionForActor!(execute=[@n4.then, @n1.then], Actor:Actor Object Reference=GetOwningPlayer()) -> (then, ReturnValue:Jam Session Interface Interface)
   n3 = TryGetKeyModeTempo!(self:Jam Session Component Music Manager Object Reference=GetMusicManager!(self:Jam Session Interface Interface=@n2) -> (ReturnValue:Jam Session Component Music Manager Object Reference)) -> (OutMusicKey:EMusicKey Enum=C, OutMusicKeyMode:EMusicKeyMode Enum=Major, OutTempo:int=0, ReturnValue:bool)
   > Branch(Condition=@n3.ReturnValue)
   > "SetText (Text)"(
    self=Get MyText() -> (MyText:Common Text Object Reference)
    InText="Get (a copy)"(Array:Array of Texts=Get Texts() -> (Texts:Array of Texts), Dimension 1="To Integer (Byte)"!(InByte:Byte=@n3) -> (ReturnValue:int=0)) -> (Output:Text)
   )
   Event Construct()
   > n4 = Parent: Construct()
   n5 = Event PreConstruct()
   > Parent: PreConstruct(IsDesignTime=@n5.IsDesignTime)
   > PopulateEntries()
  ]}
 OverrideKeyEntry {nodes [
   n1 = OverrideKeyEntry!() -> (then, MusicKeyPreview:EMusicKey Enum=C) {is=title K2Node_FunctionEntry}
   > "SetText (Text)"(self=Get MyText(), InText="Get (a copy)"(Array=Get Texts(), Dimension 1="To Integer (Byte)"(InByte=@n1) -> (ReturnValue=0)))
  ]}
 PopulateEntries {nodes [
   PopulateEntries() {is=title K2Node_FunctionEntry}
   > n1 = ForEach EMusicKey!(execute:Exec, SkipHidden:bool) -> (LoopBody:Exec, EnumValue:EMusicKey Enum, then:Exec) {is=title K2Node_ForEachElementInEnum}
   > Branch(Condition="Not Equal (Enum)"!(A:EMusicKey Enum=@n1, B:EMusicKey Enum=Num) -> (ReturnValue:bool))
   > Add Unique(
    TargetArray:Array of Texts=Get Texts()
    NewItem:"Text (by ref)"=Select!(
     C:Text=Format Text(Format="LOCTEXT(\"C\")")
     Db:Text=Format Text(Format="LOCTEXT(\"C♯/D♭\")")
     D:Text=Format Text(Format="LOCTEXT(\"D\")")
     Eb:Text=Format Text(Format="LOCTEXT(\"D♯/E♭\")")
     E:Text=Format Text(Format="LOCTEXT(\"E\")")
     F:Text=Format Text(Format="LOCTEXT(\"F\")")
     Gb:Text=Format Text(Format="LOCTEXT(\"F♯/G♭\")")
     G:Text=Format Text(Format="LOCTEXT(\"G\")")
     Ab:Text=Format Text(Format="LOCTEXT(\"G♯/A♭\")")
     A:Text=Format Text(Format="LOCTEXT(\"A\")")
     Bb:Text=Format Text(Format="LOCTEXT(\"A♯/B♭\")")
     B:Text=Format Text(Format="LOCTEXT(\"B\")")
     Num:Text
     Index:EMusicKey Enum=@n1
    ) -> (ReturnValue:Text)
   )
  ]}
]
variables [Texts {type="array<string>", default []}]
widgets {
 info {inheritedWidgetCount=0, parentClass=&UserWidget, rootWidgetClass=&GridPanel, widgetCount=3}
 widgets [
  Background {
   widgetClassPath=&GridPanel
   bIsVariable=true
   widget=Background
   properties {bOverride_Cursor=true, visibility=Visible, columnFill [1], rowFill [1]}
  }
  MyText {
   widgetClassPath=&CommonTextBlock
   bIsVariable=true
   parent=Background
   widget=MyText
   properties {
    justification=Center
    scrollStyle=&TextScrollStyle-Base_C
    text="{Key}"
    textTransformPolicy=ToUpper
    font {
     bForceMonospaced=false
     bMaterialIsStencil=false
     fontMaterial=null
     fontObject=&HeadingNow
     letterSpacing=0
     monospacedWidth=1
     size=16
     skewAmount=0
     typefaceFontName=76 Bold
     outlineSettings {
      bApplyOutlineToDropShadows=false
      bMiteredCorners=false
      bSeparateFillAlpha=false
      outlineMaterial=null
      outlineSize=0
      outlineColor=<rgba 0 0 0 1>
     }
    }
    shadowOffset=<0 0>
   }
   slot {class=GridSlot, properties {horizontalAlignment=HAlign_Center}}
  }
  Spacer_38 {
   widgetClassPath=&Spacer
   parent=Background
   widget=Spacer_38
   properties {size=<70 1>}
   slot {class=GridSlot, properties {verticalAlignment=VAlign_Bottom}}
  }
 ]
}
