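# WBP_FMJamLoopButton is a loop button for the Jam global controls, derived from the UIKit regular button with no widgets of its own.
# GetButtonSongInfo returns title, artist and loop type from the button's Jam Music Slot view model.
# SetCanReset toggles the button's enabled state, and SetMyText sets its label.
# The event graph is empty, and the SongShortName variable is unused in these graphs.
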WidgetBlueprint /FMJamSystem/Widgets/GlobalControls/WBP_FMJamLoopButton
graphs [
 EventGraph {links [], nodes []}
 GetButtonSongInfo {nodes [
   GetButtonSongInfo() {is=title K2Node_FunctionEntry}
   n1 = Get JamMusicSlotVM() -> (JamMusicSlotVM:Jam Music Slot VM Object Reference)
   > Return Node!(
    execute
    Title:Text=Get Title!(self:Jam Music Slot VM Object Reference=@n1) -> (Title:Text)
    Artist:Text=Get Artist!(self:Jam Music Slot VM Object Reference=@n1) -> (Artist:Text)
    LoopType:EFMJamLoopType Enum=Get LoopType!(self:Jam Music Slot VM Object Reference=@n1) -> (LoopType:EFMJamLoopType Enum=Lead)
   )
  ]}
 SetCanReset {nodes [
   n1 = SetCanReset!() -> (then, CanReset:bool=~) {is=title K2Node_FunctionEntry}
   > SetIsEnabled(bInIsEnabled=@n1)
  ]}
 SetMyText {nodes [
   n1 = SetMyText!() -> (then, InText:Text) {is=title K2Node_FunctionEntry}
   > SetText!(execute, self:Fort CTAButton Object Reference, ButtonText:"Text (by ref)"=@n1) -> (then)
  ]}
]
variables [SongShortName {type=string, default=None}]
widgets {
 info {
  inheritedWidgetCount=0
  parentClass=/FortUILibrary/UIKit/Buttons/WBP_UIKit_Button_Regular.WBP_UIKit_Button_Regular_C
  rootWidgetClass=null
  widgetCount=0
 }
 widgets []
}
