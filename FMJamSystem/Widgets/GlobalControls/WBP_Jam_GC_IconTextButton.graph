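# Jam UI button widget derived from SimpleCommonButton, with an optional icon, styled centre text, a named extra-content slot and an input action widget.
# Input-method changes swap between controller and mouse/keyboard styles (controller always on Switch) and refresh text, preferring action-widget text, then override, then default.
# PreConstruct applies brush, font, casing and alignment, collapsing empty parts.
# Hover and click animate the icon, while press, release and focus are re-broadcast through dispatchers.
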WidgetBlueprint /FMJamSystem/Widgets/GlobalControls/WBP_Jam_GC_IconTextButton
animations [
 A_Hover {
  binding_count=1
  length_seconds=0.1
  object=&self:A_Hover
  sequence {bindings [LeftSideImage {
     id=#1
     possessed_class=Image
     tracks [{
       type=MovieScene2DTransformTrack
       sections [{
         type=MovieScene2DTransformSection
         end_frame=2
         start_frame=0
         channels [
          Translation.X {type=MovieSceneScriptingFloatChannel}
          Translation.Y {type=MovieSceneScriptingFloatChannel}
          Angle {type=MovieSceneScriptingFloatChannel}
          Scale.X {type=MovieSceneScriptingFloatChannel, keys [[0,1],[2,1.2]]}
          Scale.Y {type=MovieSceneScriptingFloatChannel, keys [[0,1],[2,1.2]]}
          Shear.X {type=MovieSceneScriptingFloatChannel}
          Shear.Y {type=MovieSceneScriptingFloatChannel}
         ]
        }]
      }]
    }]}
 }
 A_Activate {
  binding_count=1
  length_seconds=0.1
  object=&self:A_Activate
  sequence {bindings [LeftSideImage {
     id=#2
     possessed_class=Image
     tracks [{
       type=MovieScene2DTransformTrack
       sections [{
         type=MovieScene2DTransformSection
         end_frame=2
         start_frame=0
         channels [
          Translation.X {type=MovieSceneScriptingFloatChannel, const [0,3]}
          Translation.Y {type=MovieSceneScriptingFloatChannel, keys [[0,0],[1,2],[2,0]]}
          Angle {type=MovieSceneScriptingFloatChannel}
          Scale.X {type=MovieSceneScriptingFloatChannel}
          Scale.Y {type=MovieSceneScriptingFloatChannel}
          Shear.X {type=MovieSceneScriptingFloatChannel}
          Shear.Y {type=MovieSceneScriptingFloatChannel}
         ]
        }]
      }]
    }]}
 }
]
graphs [
 BindCommonInputSubsystem {nodes [
   BindCommonInputSubsystem!() -> (then, bBind:bool=~) {is=title K2Node_FunctionEntry}
   n1 = Get CommonInputSubsystem!() -> (ReturnValue:Common Input Object Reference) {is=title K2Node_GetBlueprintContextEx}
   > Is Valid(InputObject=@n1)
   > n2 = Branch(Condition=Get bBind() -> (bBind:bool))
   n3 = Create Event()
   > Bind Event to On Input Method Changed(self=@n1, Delegate=@n3)
   > UpdateTextAndStyle!(execute, self, InputType:ECommonInputType Enum=GetCurrentInputType(self=@n1)) -> (then)
   Unbind Event from On Input Method Changed(execute=@n2.else, self=@n1, Delegate=@n3)
  ]}
 EventGraph {nodes [
   n1 = Event OnTriggeringInputActionChanged!() -> (OutputDelegate, then, NewTriggeredAction:"Data Table Row Handle Structure (by ref)") {is=title K2Node_Event}
   UpdateTextAndStyle(execute=[@n1.then, @n6.then], InputType=GetCurrentInputType(self=Get CommonInputSubsystem()))
   Event OnFocusLost!() -> (OutputDelegate, then, InFocusEvent:Focus Event Structure) {is=title K2Node_Event}
   > Call Focus Changed Event!(execute:Exec, self:WBP Jam GC Icon Text Button Object Reference, HasFocus:bool) -> (then:Exec) {is=title K2Node_CallDelegate}
   Event On Pressed() {is=title K2Node_Event}
   > Call On Button Pressed!(execute:Exec, self:WBP Jam GC Icon Text Button Object Reference, Button:Common Button Legacy Object Reference=Self-Reference()) -> (then:Exec) {is=title K2Node_CallDelegate}
   n2 = Event SetTabLabelInfo!() -> (OutputDelegate, then, TabLabelInfo:"Fort Tab Button Label Info Structure (by ref)") {is=title K2Node_Event}
   > n3 = Sequence()
   n4 = Break Fort Tab Button Label Info!(FortTabButtonLabelInfo:"Fort Tab Button Label Info Structure (by ref)"=@n2.TabLabelInfo) -> (DisplayName:Text, IconBrush:Slate Brush Structure, IconResource:Soft Object Reference) {is=title K2Node_BreakStruct}
   > Set Text!(execute, self, ButtonText:Text=@n4) -> (then) {is=title K2Node_CallFunction}
   Set Icon!(execute=@n3.then_1, self, IconBrush:Slate Brush Structure=@n4.IconBrush) -> (then) {is=title K2Node_CallFunction}
   Event OnCurrentTextStyleChanged() {is=title K2Node_Event}
   > UpdateTextStyle()
   Event Destruct()
   > Set OverrideButtonText(OverrideButtonText:Text=GetEmptyText!() -> (ReturnValue:Text)) -> (Output_Get:Text)
   > BindCommonInputSubsystem!(execute, self, bBind:bool=false) -> (then) {is=title K2Node_CallFunction}
   n5 = Event On Disabled() {is=title K2Node_Event}
   SetRenderOpacity(
    execute=[@n5.then, @n8.then]
    self=Get LeftSideImage() -> (LeftSideImage:Image Object Reference)
    InOpacity=SelectFloat(A=1, B=0.6, bPickA=GetIsEnabled!(self:Widget Object Reference) -> (ReturnValue:bool))
   )
   n6 = Event OnTriggeredInputActionChanged!() -> (OutputDelegate, then, NewTriggeredAction:"Data Table Row Handle Structure (by ref)") {is=title K2Node_Event}
   n7 = Event OnActionProgress!() -> (OutputDelegate, then, HeldPercent:float) {is=title K2Node_Event}
   > Set PressProgress(PressProgress:double=@n7.HeldPercent) -> (Output_Get:double)
   Event Construct()
   > SetInitialMouseKeyboardStyle()
   > InitializeButton()
   > BindCommonInputSubsystem(bBind=true)
   n8 = Event On Enabled() {is=title K2Node_Event}
   n9 = Event PreConstruct()
   > n10 = Sequence!(execute) -> (then_0, then_1, then_2:Exec, then_3:Exec)
   "SetText (Text)"(
    execute=@n10.then_2
    self=Get CenterButtonTextWidget() -> (CenterButtonTextWidget:Common Text Object Reference)
    InText=Get ButtonText() -> (ButtonText:Text)
   )
   > UpdateTextStyle()
   > SetTextCase!(execute, self:Common Text Object Reference=Get CenterButtonTextWidget(), bUseAllCaps:bool=Get bDisplayAllCaps() -> (bDisplayAllCaps:bool)) -> (then)
   > SetFont(self=Get CenterButtonTextWidget(), InFontInfo=Get Font() -> (Font:Slate Font Info Structure))
   UpdateContentAlignment(execute=@n10.then_0)
   Branch(execute=@n10.then_3, Condition=@n9.IsDesignTime)
   n11 = Get RightExtraContentSlot() -> (RightExtraContentSlot:Named Slot Object Reference)
   > SetVisibility(
    self=@n11
    InVisibility=Select(Option 0=Collapsed, Option 1=SelfHitTestInvisible, Index=HasAnyChildren!(self:Panel Widget Object Reference=@n11) -> (ReturnValue:bool))
   )
   > InitializeButton()
   n12 = Get IconBrush() -> (IconBrush:Slate Brush Structure)
   SetBrush(execute=@n10.then_1, self=Get LeftSideImage(), InBrush=@n12)
   n13 = Get Class Defaults!(Class:Object Class Reference=/FortUILibrary/Foundation/Buttons/IconTextButton.IconTextButton_C) -> (
    ButtonText:Text
    IconBrush:Slate Brush Structure
    ControllerInputStyle:Common Button Style Class Reference
    MouseKeyboardStyle:Common Button Style Class Reference
    ContentAlignment:EHorizontalAlignment Enum
    bMouseKeyboardStyleSet:bool=~
    OverrideButtonText:Text
    ButtonTextJustification:ETextJustify Enum
    bDisplayAllCaps:bool=~
    PressProgress:double=~
    bIgnoreInputActionWidgetText:bool=~
    FocusChangedEvent:Multicast Delegate
    Font:Slate Font Info Structure
    UseLightInput:bool=~
   ) {is=title K2Node_GetClassDefaults}
   > SetVisibility(
    self=Get LeftSideImage()
    InVisibility=Select(
     Option 0=Visible
     Option 1=Collapsed
     Index="Equal (SlateBrush)"!(A:"Slate Brush Structure (by ref)"=@n12, B:"Slate Brush Structure (by ref)"=@n13.IconBrush) -> (ReturnValue:bool)
    )
   )
   Event OnMouseEnter()
   > PlayAnimationForward(InAnimation=Get A_Hover() -> (A_Hover:Widget Animation Object Reference))
   Event OnMouseLeave!() -> (OutputDelegate, then, MouseEvent:"Pointer Event Structure (by ref)") {is=title K2Node_Event}
   > PlayAnimationReverse(InAnimation=Get A_Hover())
   Event On Clicked()
   > PlayAnimationForward(InAnimation=Get A_Activate() -> (A_Activate:Widget Animation Object Reference), bRestoreState=true)
   Event On Released() {is=title K2Node_Event}
   > Call On Button Released!(execute:Exec, self:WBP Jam GC Icon Text Button Object Reference, Button:Common Button Legacy Object Reference=Self-Reference()) -> (then:Exec) {is=title K2Node_CallDelegate}
   Event OnActionComplete() {is=title K2Node_Event}
  ]}
 FocusChangedEvent {nodes [FocusChangedEvent!() -> (then, HasFocus:bool=~) {is=title K2Node_FunctionEntry}]}
 Get Dynamic Material {nodes [
   Get Dynamic Material() {is=title K2Node_FunctionEntry}
   > n1 = GetDynamicMaterial(self=Get LeftSideImage())
   > Return Node!(execute, Ret Material:Material Instance Dynamic Object Reference=@n1)
  ]}
 GetPressProgress {nodes [
   GetPressProgress() {is=title K2Node_FunctionEntry}
   > Return Node!(execute, Progress:double=Get PressProgress() -> (PressProgress:double))
  ]}
 InitializeButton {nodes [
   InitializeButton() {is=title K2Node_FunctionEntry}
   n1 = Get IconBrush()
   > n2 = Is Valid(InputObject=GetBrushResource!(Brush:"Slate Brush Structure (by ref)"=@n1) -> (ReturnValue:Object Reference))
   > n3 = Branch(Condition="Not Equal (Enum)"!(
     A:ESlateBrushDrawType Enum=Break Slate Brush!(SlateBrush:"Slate Brush Structure (by ref)"=@n1) -> (
      TintColor:Slate Color Structure
      DrawAs:ESlateBrushDrawType Enum
      Tiling:ESlateBrushTileType Enum
      Mirroring:ESlateBrushMirrorType Enum
      ImageSize:Vector 2D Structure
      Margin:Margin Structure
      ResourceObject:Object Reference
      OutlineSettings:Slate Brush Outline Settings Structure
     ) {is=title K2Node_BreakStruct}.DrawAs
     B:ESlateBrushDrawType Enum=NoDrawType
    ) -> (ReturnValue:bool))
   > SetBrush(self=Get LeftSideImage(), InBrush=Get IconBrush())
   SetVisibility(execute=[@n3.else, @n2.Is Not Valid], self=Get LeftSideImage(), InVisibility=Collapsed)
  ]}
 OnButtonPressed {nodes [OnButtonPressed!() -> (then, Button:Common Button Legacy Object Reference) {is=title K2Node_FunctionEntry}]}
 OnButtonReleased {nodes [OnButtonReleased!() -> (then, Button:Common Button Legacy Object Reference) {is=title K2Node_FunctionEntry}]}
 OnFocusReceived {nodes [
   OnFocusReceived!() -> (then, MyGeometry:Geometry Structure, InFocusEvent:Focus Event Structure) {is=title K2Node_FunctionEntry}
   > Call Focus Changed Event(HasFocus=true)
   > Return Node!(execute, ReturnValue:Event Reply Structure=Make Event Reply!() -> (EventReply:Event Reply Structure) {is=title K2Node_MakeStruct})
  ]}
 Set Icon {nodes [
   n1 = Set Icon!() -> (then, IconBrush:Slate Brush Structure) {is=title K2Node_FunctionEntry}
   > n2 = Set IconBrush(IconBrush:Slate Brush Structure=@n1) -> (Output_Get:Slate Brush Structure)
   > SetBrush(self=Get LeftSideImage(), InBrush=@n2)
  ]}
 Set Text {nodes [
   n1 = Set Text!() -> (then, ButtonText:Text) {is=title K2Node_FunctionEntry}
   > n2 = Set OverrideButtonText(OverrideButtonText=@n1)
   > SetTextInternal!(execute, self, InButtonText:Text=@n2) -> (then)
  ]}
 SetContentAlignment {nodes [
   n1 = SetContentAlignment!() -> (then, ContentAlignment:EHorizontalAlignment Enum=HAlign_Fill) {is=title K2Node_FunctionEntry}
   > Set ContentAlignment(ContentAlignment:EHorizontalAlignment Enum=@n1) -> (Output_Get:EHorizontalAlignment Enum=HAlign_Fill)
   > UpdateContentAlignment()
  ]}
 SetControllerStyle {nodes [
   n1 = SetControllerStyle!() -> (then, ControllerInputStyle:Common Button Style Class Reference) {is=title K2Node_FunctionEntry}
   > Set ControllerInputStyle(ControllerInputStyle:Common Button Style Class Reference=@n1) -> (Output_Get:Common Button Style Class Reference)
   n2 = Get CommonInputSubsystem()
   > UpdateStyle!(execute, self, UsingGamepad:bool="Equal (Enum)"(A=GetCurrentInputType(self=@n2), B=Gamepad)) -> (then)
  ]}
 SetInitialMouseKeyboardStyle {nodes [
   SetInitialMouseKeyboardStyle() {is=title K2Node_FunctionEntry}
   > n1 = Branch(Condition=Get bMouseKeyboardStyleSet() -> (bMouseKeyboardStyleSet:bool))
   Set bMouseKeyboardStyleSet(execute=@n1.else, bMouseKeyboardStyleSet:bool=true) -> (Output_Get:bool)
   > Set MouseKeyboardStyle(MouseKeyboardStyle:Common Button Style Class Reference=Get Style() -> (Style:Common Button Style Class Reference)) -> (Output_Get:Common Button Style Class Reference)
  ]}
 SetMouseKeyboardStyle {nodes [
   n1 = SetMouseKeyboardStyle!() -> (then, ControllerInputStyle:Common Button Style Class Reference) {is=title K2Node_FunctionEntry}
   > Set MouseKeyboardStyle(MouseKeyboardStyle=@n1)
   n2 = Get CommonInputSubsystem()
   > UpdateStyle(UsingGamepad="Equal (Enum)"(A=GetCurrentInputType(self=@n2), B=Gamepad))
  ]}
 SetTextInternal {nodes [
   SetTextInternal!() -> (then, InButtonText:Text) {is=title K2Node_FunctionEntry}
   n1 = Get CenterButtonTextWidget()
   > "SetText (Text)"(self=@n1, InText=Get InButtonText() -> (InButtonText:Text))
   n2 = Get InButtonText()
   > SetVisibility(self=@n1, InVisibility=Select(Option 0=SelfHitTestInvisible, Option 1=Collapsed, Index=TextIsEmpty(InText=@n2)))
  ]}
 ShowIcon {nodes [
   n1 = ShowIcon!() -> (then, bShouldShow:bool=true) {is=title K2Node_FunctionEntry}
   > n2 = Branch(Condition=@n1)
   n3 = Get LeftSideImage()
   > SetVisibility(self=@n3, InVisibility=SelfHitTestInvisible)
   SetVisibility(execute=@n2.else, self=@n3, InVisibility=Collapsed)
  ]}
 UpdateContentAlignment {nodes [
   UpdateContentAlignment() {is=title K2Node_FunctionEntry}
   > n1 = Cast To BorderSlot!(
    execute
    Object=GetContentSlot!(self:Content Widget Object Reference=Get ContentBorder() -> (ContentBorder:Border Object Reference)) -> (ReturnValue:Panel Slot Object Reference)
   ) -> (then, CastFailed, AsBorder Slot:Border Slot Object Reference) {is=title K2Node_DynamicCast}
   > SetHorizontalAlignment(
    self:Border Slot Object Reference=@n1
    InHorizontalAlignment=Get ContentAlignment() -> (ContentAlignment:EHorizontalAlignment Enum=HAlign_Fill)
   )
   > SetJustification!(
    execute
    self:Text Layout Widget Object Reference=Get CenterButtonTextWidget()
    InJustification:ETextJustify Enum=Get ButtonTextJustification() -> (ButtonTextJustification:ETextJustify Enum=Left)
   ) -> (then)
  ]}
 UpdateStyle {nodes [
   n1 = UpdateStyle!() -> (then, UsingGamepad:bool=~) {is=title K2Node_FunctionEntry}
   > n2 = Branch(Condition=OR Boolean(A=@n1, B=IsNintendoSwitch!() -> (ReturnValue:bool)))
   > SetStyle!(
    execute
    self:Common Button Base Object Reference
    InStyle:Common Button Style Class Reference=Get ControllerInputStyle() -> (ControllerInputStyle:Common Button Style Class Reference)
   ) -> (then)
   SetStyle(execute=@n2.else, InStyle=Get MouseKeyboardStyle() -> (MouseKeyboardStyle:Common Button Style Class Reference))
  ]}
 UpdateText {nodes [
   UpdateText() {is=title K2Node_FunctionEntry}
   n1 = NOT Boolean(A=Get bIgnoreInputActionWidgetText() -> (bIgnoreInputActionWidgetText:bool))
   n2 = GetDisplayText!(self:Common Action Widget Object Reference=Get InputActionWidget() -> (InputActionWidget:Common Action Widget Object Reference)) -> (ReturnValue:Text)
   n3 = NOT Boolean(A=TextIsEmpty(InText=@n2))
   n4 = Get OverrideButtonText() -> (OverrideButtonText:Text)
   n5 = TextIsEmpty(InText=@n4)
   n6 = Get ButtonText()
   > SetTextInternal(InButtonText=Select(Option 0:Text=Select(Option 0=@n4, Option 1=@n6, Index=@n5), Option 1:Text=@n2, Index=AND Boolean(A=@n3, B=@n1)) -> (ReturnValue:Text)) {is=title K2Node_CallFunction}
  ]}
 UpdateTextAndStyle {nodes [
   n1 = UpdateTextAndStyle!() -> (then, InputType:ECommonInputType Enum) {is=title K2Node_FunctionEntry}
   > UpdateStyle(UsingGamepad="Equal (Enum)"(A=@n1, B=Gamepad)) {is=title K2Node_CallFunction}
   > UpdateTextStyle()
   > UpdateText() {is=title K2Node_CallFunction}
  ]}
 UpdateTextStyle {nodes [
   UpdateTextStyle() {is=title K2Node_FunctionEntry}
   > SetStyle(
    self:Common Text Object Reference=Get CenterButtonTextWidget()
    InStyle:Common Text Style Class Reference=GetCurrentTextStyleClass!(self:Common Button Base Object Reference) -> (ReturnValue:Common Text Style Class Reference)
   )
  ]}
]
variables [
 bDisplayAllCaps {type=boolean, default=true}
 bIgnoreInputActionWidgetText {type=boolean}
 bMouseKeyboardStyleSet {type=boolean}
 ButtonText {type=string, default="{Text}"}
 ButtonTextJustification {type=ETextJustify, default=Center}
 ContentAlignment {type=EHorizontalAlignment, default=HAlign_Fill}
 ControllerInputStyle {type=Class, default=null}
 FocusChangedEvent {}
 Font {
  type=SlateFontInfo
  default {
   bForceMonospaced=false
   bMaterialIsStencil=false
   fontMaterial=null
   fontObject=&NotoSans
   letterSpacing=0
   monospacedWidth=1
   size=18
   skewAmount=0
   typefaceFontName=NotoSans-Bold
   outlineSettings {
    bApplyOutlineToDropShadows=false
    bMiteredCorners=false
    bSeparateFillAlpha=false
    outlineMaterial=null
    outlineSize=0
    outlineColor=<rgba 0 0 0 1>
   }
  }
 }
 IconBrush {
  type=SlateBrush
  default {
   bIsDynamicallyLoaded=false
   drawAs=NoDrawType
   imageType=NoImage
   mirroring=NoMirror
   resourceName=None
   resourceObject=null
   tiling=NoTile
   imageSize=<0 0>
   margin=<ltrb 0 0 0 0>
   outlineSettings {
    bUseBrushTransparency=false
    roundingType=HalfHeightRadius
    width=0
    color {colorUseRule=UseColor_Specified, specifiedColor=<rgba 0 0 0 0>}
    cornerRadii=<0 0 0 1>
   }
   tintColor {colorUseRule=UseColor_Specified, specifiedColor=<rgba 1 1 1 1>}
   uVRegion {bIsValid=false, max=<0 0>, min=<0 0>}
  }
 }
 MouseKeyboardStyle {type=Class, default=/FortUILibrary/Foundation/Buttons/ButtonStyles/Primary/ButtonStyle-Primary-M.ButtonStyle-Primary-M_C}
 OnButtonPressed {}
 OnButtonReleased {}
 OverrideButtonText {type=string, default=""}
 PressProgress {type=number}
]
widgets {
 info {
  inheritedWidgetCount=1
  parentClass=/Game/Athena/UI/Matchmaking/SimpleCommonButton.SimpleCommonButton_C
  rootWidgetClass=&Border
  widgetCount=7
 }
 widgets [
  ContentBorder {
   widgetClassPath=&Border
   bIsVariable=true
   widget=ContentBorder
   properties {
    background {
     bIsDynamicallyLoaded=false
     drawAs=NoDrawType
     imageType=NoImage
     mirroring=NoMirror
     resourceName=None
     resourceObject=null
     tiling=NoTile
     imageSize=<32 32>
     margin=<ltrb 0 0 0 0>
     outlineSettings {
      bUseBrushTransparency=false
      roundingType=HalfHeightRadius
      width=0
      color {colorUseRule=UseColor_Specified, specifiedColor=<rgba 0 0 0 0>}
      cornerRadii=<0 0 0 0>
     }
     tintColor {colorUseRule=UseColor_Specified, specifiedColor=<rgba 1 1 1 1>}
     uVRegion {bIsValid=false, max=<0 0>, min=<0 0>}
    }
    padding=<ltrb 0 0 0 0>
   }
  }
  HorizontalBox_ButtonContent {
   widgetClassPath=&HorizontalBox
   bIsVariable=true
   parent=ContentBorder
   widget=HorizontalBox_ButtonContent
   properties {flowDirectionPreference=Culture}
   slot {class=BorderSlot, properties {padding=<ltrb 0 0 0 0>}}
  }
  Overlay_0 {
   widgetClassPath=&Overlay
   parent=HorizontalBox_ButtonContent
   widget=Overlay_0
   slot {class=HorizontalBoxSlot, properties {size {sizeRule=Fill, value=1}}}
  }
  RightExtraContentSlot {
   widgetClassPath=&NamedSlot
   bIsVariable=true
   parent=Overlay_0
   widget=RightExtraContentSlot
   properties {bExposeOnInstanceOnly=true}
   slot {class=OverlaySlot, properties {horizontalAlignment=HAlign_Fill, verticalAlignment=VAlign_Fill}}
  }
  CenterButtonTextWidget {
   widgetClassPath=&CommonTextBlock
   bIsVariable=true
   parent=Overlay_0
   widget=CenterButtonTextWidget
   properties {
    justification=Center
    mobileFontSizeMultiplier=1.15
    scrollStyle=&TextScrollStyle-Base_C
    style=/FortUILibrary/Foundation/Buttons/TextStyles/TextStyle-MediumButton.TextStyle-MediumButton_C
    text="{Text}"
    textTransformPolicy=ToUpper
    font {
     bForceMonospaced=false
     bMaterialIsStencil=false
     fontMaterial=null
     fontObject=&NotoSans
     letterSpacing=0
     monospacedWidth=1
     size=18
     skewAmount=0
     typefaceFontName=NotoSans-Bold
     outlineSettings {
      bApplyOutlineToDropShadows=false
      bMiteredCorners=false
      bSeparateFillAlpha=false
      outlineMaterial=null
      outlineSize=0
      outlineColor=<rgba 0 0 0 1>
     }
    }
    shadowOffset=<0 0>
   }
   slot {class=OverlaySlot, properties {horizontalAlignment=HAlign_Fill, verticalAlignment=VAlign_Center, padding=<ltrb 5 4 5 4>}}
  }
  LeftSideImage {
   widgetClassPath=&Image
   bIsVariable=true
   parent=Overlay_0
   widget=LeftSideImage
   properties {brush {
     bIsDynamicallyLoaded=false
     drawAs=NoDrawType
     imageType=NoImage
     mirroring=NoMirror
     resourceName=None
     resourceObject=null
     tiling=NoTile
     imageSize=<0 0>
     margin=<ltrb 0 0 0 0>
     outlineSettings {
      bUseBrushTransparency=false
      roundingType=HalfHeightRadius
      width=0
      color {colorUseRule=UseColor_Specified, specifiedColor=<rgba 0 0 0 0>}
      cornerRadii=<0 0 0 0>
     }
     tintColor {colorUseRule=UseColor_Specified, specifiedColor=<rgba 1 1 1 1>}
     uVRegion {bIsValid=false, max=<0 0>, min=<0 0>}
    }}
   slot {class=OverlaySlot, properties {horizontalAlignment=HAlign_Center, verticalAlignment=VAlign_Center, padding=<ltrb 4 4 4 4>}}
  }
  InputActionWidget {
   widgetClassPath=&CommonActionWidget
   bInherited=true
   bIsVariable=true
   parent=Overlay_0
   widget=InputActionWidget
   slot {class=OverlaySlot, properties {horizontalAlignment=HAlign_Fill, verticalAlignment=VAlign_Fill}}
  }
 ]
}
