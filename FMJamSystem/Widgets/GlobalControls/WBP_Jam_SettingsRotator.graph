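# Settings rotator widget for the Jam system UI, a visual layout for the native JamSettingRotator class with no Blueprint logic of its own.
# An Overlay stretches a regular UIKit button across the widget as the option value display, with a "{Value}" placeholder label and a hidden input-action hint.
# The inherited MyText block is styled as centered uppercase text but collapsed, and the inherited InputActionWidget slot has no widget.
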WidgetBlueprint /FMJamSystem/Widgets/GlobalControls/WBP_Jam_SettingsRotator
graphs [EventGraph {links [], nodes []}]
variables []
widgets {
 info {inheritedWidgetCount=3, parentClass=/Script/FMJamUI.JamSettingRotator, rootWidgetClass=&Overlay, widgetCount=4}
 widgets [
  Overlay_0 {widgetClassPath=&Overlay, widget=Overlay_0}
  Button_OptionValue {
   widgetClassPath=&WBP_UIKit_Button_Regular
   bInherited=true
   bIsVariable=true
   parent=Overlay_0
   widget=Button_OptionValue
   properties {bDisplayInputActionWhenNotInteractable=false, bHideInputAction=true, bHideInputActionWithKeyboard=true, text="{Value}"}
   slot {class=OverlaySlot, properties {horizontalAlignment=HAlign_Fill, verticalAlignment=VAlign_Fill}}
  }
  MyText {
   widgetClassPath=&CommonTextBlock
   bInherited=true
   bIsVariable=true
   parent=Overlay_0
   widget=MyText
   properties {
    justification=Center
    scrollStyle=&TextScrollStyle-Base_C
    textTransformPolicy=ToUpper
    visibility=Collapsed
    colorAndOpacity {colorUseRule=UseColor_Specified, specifiedColor=<rgba 0 0 0 1>}
    font {
     bForceMonospaced=false
     bMaterialIsStencil=false
     fontMaterial=null
     fontObject=&HeadingNow
     letterSpacing=0
     monospacedWidth=1
     size=16
     skewAmount=0
     typefaceFontName=76 Bold
     outlineSettings {
      bApplyOutlineToDropShadows=false
      bMiteredCorners=false
      bSeparateFillAlpha=false
      outlineMaterial=null
      outlineSize=0
      outlineColor=<rgba 0 0 0 1>
     }
    }
    shadowOffset=<0 0>
   }
   slot {class=OverlaySlot, properties {horizontalAlignment=HAlign_Fill, verticalAlignment=VAlign_Center}}
  }
  InputActionWidget {widgetClassPath=&CommonActionWidget, bInherited=true, widget=null}
 ]
}
