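# Jam settings list entry for discrete options, built on FortSettingsListEntrySetting_Discrete, with a value rotator flanked by chevron decrease and increase buttons.
# Mouse enter focuses the rotator, which is also the primary gamepad focus target; entering or leaving the focus path fades an outline in or out.
# On construct the chevrons show only if the entry is enabled, and a Hide Text Block flag collapses the name label and spacer, previewed at design time.
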WidgetBlueprint /FMJamSystem/Widgets/GlobalControls/SettingsListEntry_JamDiscrete
animations [A_OnHover {
  binding_count=1
  length_seconds=0.1
  object=&self:A_OnHover
  sequence {bindings [FocusOutline {
     id=#1
     possessed_class=Image
     tracks [{
       type=MovieSceneFloatTrack
       display_name=Render Opacity
       sections [{type=MovieSceneFloatSection, end_frame=2, start_frame=0, channels [None {type=MovieSceneScriptingFloatChannel, keys [[0,0],[2,1]]}]}]
      }]
    }]}
 }]
graphs [
 EventGraph {nodes [
   n1 = Event OnAddedToFocusPath!() -> (OutputDelegate, then, InFocusEvent:Focus Event Structure) {is=title K2Node_Event}
   > Parent: OnAddedToFocusPath!(execute, InFocusEvent:Focus Event Structure=@n1.InFocusEvent) -> (then) {is=title K2Node_CallParentFunction}
   > PlayAnimationForward(InAnimation=Get A_OnHover() -> (A_OnHover:Widget Animation Object Reference))
   Event Construct()
   > SetIsEnabledVisual!(execute, self, IsEnabled:bool=GetIsEnabled!(self:Widget Object Reference) -> (ReturnValue:bool)) -> (then)
   > HideTextBlock!(execute, self, HideTextBlock:bool=Get Hide Text Block() -> (Hide Text Block:bool)) -> (then)
   n2 = Event OnMouseEnter()
   > Parent: OnMouseEnter!(execute, MyGeometry:Geometry Structure=@n2.MyGeometry, MouseEvent:"Pointer Event Structure (by ref)"=@n2.MouseEvent) -> (then) {is=title K2Node_CallParentFunction}
   > SetFocus(self=Get Rotator_SettingValue() -> (Rotator_SettingValue:Fort Settings Rotator Object Reference))
   n3 = Event OnRemovedFromFocusPath!() -> (OutputDelegate, then, InFocusEvent:Focus Event Structure) {is=title K2Node_Event}
   > Parent: OnRemovedFromFocusPath!(execute, InFocusEvent:Focus Event Structure=@n3.InFocusEvent) -> (then) {is=title K2Node_CallParentFunction}
   > PlayAnimationReverse(InAnimation=Get A_OnHover())
   n4 = Event PreConstruct()
   > Branch(Condition=@n4.IsDesignTime)
   > HideTextBlock(HideTextBlock=Get Hide Text Block())
  ]}
 GetPrimaryGamepadFocusWidget {nodes [
   GetPrimaryGamepadFocusWidget() {is=title K2Node_FunctionEntry}
   > Return Node!(execute, ReturnValue:Widget Object Reference=Get Rotator_SettingValue())
  ]}
 HideTextBlock {nodes [
   n1 = HideTextBlock!() -> (then, HideTextBlock:bool=~) {is=title K2Node_FunctionEntry}
   > n2 = Branch(Condition=@n1)
   > SetVisibility(
    self=[
     Get Text_SettingName() -> (Text_SettingName:Common Text Object Reference)
     Get SettingName_Spacer() -> (SettingName_Spacer:Spacer Object Reference)
    ]
    InVisibility=Collapsed
   )
   SetVisibility(execute=@n2.else, self=[Get Text_SettingName(), Get SettingName_Spacer()], InVisibility=SelfHitTestInvisible)
  ]}
 SetIsEnabledVisual {nodes [
   n1 = SetIsEnabledVisual!() -> (then, IsEnabled:bool=~) {is=title K2Node_FunctionEntry}
   > SetVisibility(
    self=[
     Get Button_Decrease() -> (Button_Decrease:Common Button Legacy Object Reference)
     Get Button_Increase() -> (Button_Increase:Common Button Legacy Object Reference)
    ]
    InVisibility=Select(Option 0=Collapsed, Option 1=SelfHitTestInvisible, Index=@n1)
   )
  ]}
]
variables [Hide Text Block {type=boolean}]
widgets {
 info {
  inheritedWidgetCount=8
  namedSlotCount=1
  parentClass=/Script/FortniteUI.FortSettingsListEntrySetting_Discrete
  rootWidgetClass=&/WBP_Jam_SettingEntryBackground.WBP_Jam_SettingEntryBackground_C
  widgetCount=22
 }
 widgets [
  WBP_Jam_SettingEntryBackground_C_0 {
   widgetClassPath=&/WBP_Jam_SettingEntryBackground
   bIsVariable=true
   widget=WBP_Jam_SettingEntryBackground_C_0
   properties {visibility=SelfHitTestInvisible}
  }
  FortResponsivePanel_13 {widgetClassPath=/Script/FortniteUI.FortResponsivePanel, namedSlotHost=WBP_Jam_SettingEntryBackground_C_0, widget=FortResponsivePanel_13}
  Overlay_0 {widgetClassPath=&Overlay, parent=FortResponsivePanel_13, widget=Overlay_0, slot {class=FortResponsivePanelSlot}}
  VerticalBox_18 {
   widgetClassPath=&VerticalBox
   parent=Overlay_0
   widget=VerticalBox_18
   slot {class=OverlaySlot, properties {horizontalAlignment=HAlign_Fill, verticalAlignment=VAlign_Fill}}
  }
  SizeBox_27 {widgetClassPath=&SizeBox, parent=VerticalBox_18, widget=SizeBox_27, slot {class=VerticalBoxSlot}}
  Text_SettingName {
   widgetClassPath=&CommonTextBlock
   bInherited=true
   bIsVariable=true
   parent=SizeBox_27
   widget=Text_SettingName
   properties {
    text=Text Block
    font {
     bForceMonospaced=false
     bMaterialIsStencil=false
     fontMaterial=null
     fontObject=&HeadingNow
     letterSpacing=0
     monospacedWidth=1
     size=16
     skewAmount=0
     typefaceFontName=75 Medium
     outlineSettings {
      bApplyOutlineToDropShadows=false
      bMiteredCorners=false
      bSeparateFillAlpha=false
      outlineMaterial=null
      outlineSize=0
      outlineColor=<rgba 0 0 0 1>
     }
    }
    shadowOffset=<0 0>
    strikeBrush {
     bIsDynamicallyLoaded=false
     drawAs=NoDrawType
     imageType=NoImage
     mirroring=NoMirror
     resourceName=None
     resourceObject=null
     tiling=NoTile
     imageSize=<32 32>
     margin=<ltrb 0 0 0 0>
     outlineSettings {
      bUseBrushTransparency=false
      roundingType=HalfHeightRadius
      width=0
      color {colorUseRule=UseColor_Specified, specifiedColor=<rgba 0 0 0 0>}
      cornerRadii=<0 0 0 0>
     }
     tintColor {colorUseRule=UseColor_Specified, specifiedColor=<rgba 1 1 1 1>}
     uVRegion {bIsValid=false, max=<0 0>, min=<0 0>}
    }
   }
   slot {class=SizeBoxSlot, properties {horizontalAlignment=HAlign_Center}}
  }
  SettingName_Spacer {
   widgetClassPath=&Spacer
   bIsVariable=true
   parent=VerticalBox_18
   widget=SettingName_Spacer
   properties {size=<1 10>}
   slot {class=VerticalBoxSlot}
  }
  ExpansionSlot {
   widgetClassPath=&NamedSlot
   bIsVariable=true
   parent=VerticalBox_18
   widget=ExpansionSlot
   properties {visibility=Collapsed}
   slot {class=VerticalBoxSlot}
  }
  SettingExpansion_Spacer {
   widgetClassPath=&Spacer
   bIsVariable=true
   parent=VerticalBox_18
   widget=SettingExpansion_Spacer
   properties {visibility=Collapsed, size=<1 10>}
   slot {class=VerticalBoxSlot}
  }
  Overlay_Value {widgetClassPath=&Overlay, bIsVariable=true, parent=VerticalBox_18, widget=Overlay_Value, slot {class=VerticalBoxSlot}}
  Panel_Value {
   widgetClassPath=&HorizontalBox
   bInherited=true
   bIsVariable=true
   parent=Overlay_Value
   widget=Panel_Value
   slot {class=OverlaySlot, properties {horizontalAlignment=HAlign_Fill, verticalAlignment=VAlign_Fill}}
  }
  Several_Options {
   widgetClassPath=&Overlay
   parent=Panel_Value
   widget=Several_Options
   slot {class=HorizontalBoxSlot, properties {size {sizeRule=Fill, value=1}}}
  }
  SizeBox_0 {
   widgetClassPath=&SizeBox
   parent=Several_Options
   widget=SizeBox_0
   properties {bOverride_WidthOverride=true, widthOverride=320}
   slot {class=OverlaySlot, properties {horizontalAlignment=HAlign_Fill}}
  }
  Rotator_SettingValue {
   widgetClassPath=&/WBP_Jam_SettingsRotator
   bInherited=true
   bIsVariable=true
   parent=SizeBox_0
   widget=Rotator_SettingValue
   properties {
    bNavigateToNextWidgetOnDisable=true
    style=/FortUILibrary/Foundation/Buttons/ButtonStyles/ButtonStyle-Empty.ButtonStyle-Empty_C
    touchMethod=PreciseTap
   }
   slot {class=SizeBoxSlot}
  }
  SizeBox_29 {
   widgetClassPath=&SizeBox
   parent=Several_Options
   widget=SizeBox_29
   properties {bOverride_HeightOverride=true, bOverride_WidthOverride=true, heightOverride=48, widthOverride=48}
   slot {class=OverlaySlot, properties {verticalAlignment=VAlign_Fill}}
  }
  Button_Decrease {
   widgetClassPath=&/WBP_Jam_GC_IconTextButton
   bInherited=true
   bIsVariable=true
   parent=SizeBox_29
   widget=Button_Decrease
   properties {
    buttonText=""
    controllerInputStyle=&ButtonStyle-Empty_C
    holdData=&CommonUI_HoldData_Base_C
    onButtonBaseClicked=(&self:WidgetTree.Button_Decrease.HandleOnButtonClicked)
    onButtonBaseDoubleClicked=(&self:WidgetTree.Button_Decrease.HandleOnButtonDoubleClicked)
    onButtonBaseHovered=(&self:WidgetTree.Button_Decrease.HandleOnButtonHovered)
    onButtonBaseUnhovered=(&self:WidgetTree.Button_Decrease.HandleOnButtonUnhovered)
    onSelectedChangedBase=(&self:WidgetTree.Button_Decrease.HandleOnSelectedChanged)
    style=&ButtonStyle-Empty_C
    touchMethod=PreciseTap
    iconBrush {
     bIsDynamicallyLoaded=false
     drawAs=Image
     imageType=NoImage
     mirroring=NoMirror
     resourceName=None
     resourceObject=/FMJamSystem/Widgets/GlobalControls/Textures/T_UI_ChevronLeft
     tiling=NoTile
     imageSize=<42 42>
     margin=<ltrb 0 0 0 0>
     outlineSettings {
      bUseBrushTransparency=false
      roundingType=HalfHeightRadius
      width=0
      color {colorUseRule=UseColor_Specified, specifiedColor=<rgba 0 0 0 0>}
      cornerRadii=<0 0 0 1>
     }
     tintColor {colorUseRule=UseColor_Specified, specifiedColor=<rgba 0.0156 0.0156 0.0156 1>}
     uVRegion {bIsValid=false, max=<0 0>, min=<0 0>}
    }
   }
   slot {class=SizeBoxSlot}
  }
  SizeBox_30 {
   widgetClassPath=&SizeBox
   parent=Several_Options
   widget=SizeBox_30
   properties {bOverride_HeightOverride=true, bOverride_WidthOverride=true, heightOverride=48, widthOverride=48}
   slot {class=OverlaySlot, properties {horizontalAlignment=HAlign_Right, verticalAlignment=VAlign_Fill}}
  }
  Button_Increase {
   widgetClassPath=&WBP_Jam_GC_IconTextButton
   bInherited=true
   bIsVariable=true
   parent=SizeBox_30
   widget=Button_Increase
   properties {
    buttonText=""
    controllerInputStyle=&ButtonStyle-Empty_C
    holdData=&CommonUI_HoldData_Base_C
    onButtonBaseClicked=(&self:WidgetTree.Button_Increase.HandleOnButtonClicked)
    onButtonBaseDoubleClicked=(&self:WidgetTree.Button_Increase.HandleOnButtonDoubleClicked)
    onButtonBaseHovered=(&self:WidgetTree.Button_Increase.HandleOnButtonHovered)
    onButtonBaseUnhovered=(&self:WidgetTree.Button_Increase.HandleOnButtonUnhovered)
    onSelectedChangedBase=(&self:WidgetTree.Button_Increase.HandleOnSelectedChanged)
    style=&ButtonStyle-Empty_C
    touchMethod=PreciseTap
    iconBrush {
     bIsDynamicallyLoaded=false
     drawAs=Image
     imageType=NoImage
     mirroring=NoMirror
     resourceName=None
     resourceObject=/FMJamSystem/Widgets/GlobalControls/Textures/T_UI_ChevronRight
     tiling=NoTile
     imageSize=<42 42>
     margin=<ltrb 0 0 0 0>
     outlineSettings {
      bUseBrushTransparency=false
      roundingType=HalfHeightRadius
      width=0
      color {colorUseRule=UseColor_Specified, specifiedColor=<rgba 0 0 0 0>}
      cornerRadii=<0 0 0 1>
     }
     tintColor {colorUseRule=UseColor_Specified, specifiedColor=<rgba 0.0052 0.0052 0.0052 1>}
     uVRegion {bIsValid=false, max=<0 0>, min=<0 0>}
    }
   }
   slot {class=SizeBoxSlot}
  }
  FocusOutline {
   widgetClassPath=&Image
   bIsVariable=true
   parent=Overlay_Value
   widget=FocusOutline
   properties {
    pixelSnapping=Disabled
    renderOpacity=0
    visibility=SelfHitTestInvisible
    brush {
     bIsDynamicallyLoaded=false
     drawAs=Image
     imageType=NoImage
     mirroring=NoMirror
     resourceName=None
     resourceObject=/FortniteCoreUI/Settings/Art/MI_UIKit_Settings_Outline
     tiling=NoTile
     imageSize=<32 32>
     margin=<ltrb 0 0 0 0>
     outlineSettings {
      bUseBrushTransparency=false
      roundingType=HalfHeightRadius
      width=0
      color {colorUseRule=UseColor_Specified, specifiedColor=<rgba 0 0 0 0>}
      cornerRadii=<0 0 0 0>
     }
     tintColor {colorUseRule=UseColor_Specified, specifiedColor=<rgba 1 1 1 1>}
     uVRegion {bIsValid=false, max=<0 0>, min=<0 0>}
    }
   }
   slot {class=OverlaySlot, properties {horizontalAlignment=HAlign_Fill, verticalAlignment=VAlign_Fill, padding=<ltrb -16 -16 -16 -16>}}
  }
  Button_SelectEntry {widgetClassPath=&CommonButtonLegacy, bInherited=true, widget=null}
  OptionsModifiedIndicator {widgetClassPath=&Image, bInherited=true, widget=null}
  Background {widgetClassPath=&UserWidget, bInherited=true, widget=null}
 ]
}
