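# Settings-list rotator widget for choosing the Major or Minor key mode in FM Jam, derived from SettingsListEntry_JamDiscrete.
# PreConstruct fills the rotator from the EMusicKeyMode enum, and the decrease and increase buttons step through it.
# Construct and Refresh select the mode reported by the owning player's jam session music manager, falling back to Minor if that query fails.
# Rotation broadcasts OnModeChanged, and focus-path changes play FocusAnimation forward or reverse.
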WidgetBlueprint /FMJamSystem/Widgets/GlobalControls/WBP_FMJamModeSelector
graphs [
 EventGraph {nodes [
   "On Rotated (Rotator_SettingValue)"!() -> (OutputDelegate, then, Value:int=0) {is=title K2Node_ComponentBoundEvent}
   > Call On Mode Changed!(
    execute:Exec
    self:WBP FMJam Mode Selector Object Reference
    NewMode:EMusicKeyMode Enum=GetSelectedMode!() -> (Mode:EMusicKeyMode Enum=Major)
   ) -> (then:Exec) {is=title K2Node_CallDelegate}
   n1 = Refresh() {is=title K2Node_CustomEvent}
   n2 = GetJamSessionForActor!(execute=[@n8.then, @n1.then], Actor:Actor Object Reference=GetOwningPlayer()) -> (then, ReturnValue:Jam Session Interface Interface)
   n3 = TryGetKeyModeTempo!(self:Jam Session Component Music Manager Object Reference=GetMusicManager!(self:Jam Session Interface Interface=@n2) -> (ReturnValue:Jam Session Component Music Manager Object Reference)) -> (OutMusicKey:EMusicKey Enum=C, OutMusicKeyMode:EMusicKeyMode Enum=Major, OutTempo:int=0, ReturnValue:bool)
   > n4 = Branch(Condition=@n3.ReturnValue)
   > UpdateMode!(execute, self, Mode:EMusicKeyMode Enum=@n3.OutMusicKeyMode) -> (then)
   UpdateMode(execute=@n4.else, Mode=Minor)
   OnDecreaseButtonClicked_Event!() -> (OutputDelegate, then, Button:Common Button Legacy Object Reference) {is=title K2Node_CustomEvent}
   > ShiftTextLeft(self:Common Rotator Object Reference=Get Rotator_SettingValue() -> (Rotator_SettingValue:Fort Settings Rotator Object Reference))
   OnIncreaseButtonClicked_Event!() -> (OutputDelegate, then, Button:Common Button Legacy Object Reference) {is=title K2Node_CustomEvent}
   > ShiftTextRight(self:Common Rotator Object Reference=Get Rotator_SettingValue())
   n5 = UpdateMode!() -> (OutputDelegate, then, Mode:EMusicKeyMode Enum=Major) {is=title K2Node_CustomEvent}
   > SetSelectedItem!(
    execute
    self:Common Rotator Object Reference=Get Rotator_SettingValue()
    InValue:int="To Integer (Byte)"!(InByte:Byte=@n5.Mode) -> (ReturnValue:int=0)
   ) -> (then)
   Event OnAddedToFocusPath!() -> (OutputDelegate, then, InFocusEvent:Focus Event Structure) {is=title K2Node_Event}
   > Parent: OnAddedToFocusPath!(execute, InFocusEvent:Focus Event Structure) -> (then) {is=title K2Node_CallParentFunction}
   n6 = Get FocusAnimation() -> (FocusAnimation:Widget Animation Object Reference)
   > Is Valid(InputObject=@n6)
   > PlayAnimationForward(InAnimation=@n6)
   Event OnRemovedFromFocusPath!() -> (OutputDelegate, then, InFocusEvent:Focus Event Structure) {is=title K2Node_Event}
   > Parent: OnRemovedFromFocusPath!(execute, InFocusEvent:Focus Event Structure) -> (then) {is=title K2Node_CallParentFunction}
   n7 = Get FocusAnimation()
   > Is Valid(InputObject=@n7)
   > PlayAnimationReverse(InAnimation=@n7)
   Event Construct()
   > Parent: Construct()
   > Assign On Button Clicked!(
    execute:Exec
    self:Common Button Legacy Object Reference=Get Button_Decrease() -> (Button_Decrease:Common Button Legacy Object Reference)
    Delegate:"Delegate (by ref)"=Create Event()
   ) -> (then:Exec) {is=title K2Node_AssignDelegate}
   > n8 = Bind Event to On Button Clicked(
    self:Common Button Legacy Object Reference=Get Button_Increase() -> (Button_Increase:Common Button Legacy Object Reference)
    Delegate=Create Event()
   ) {is=title K2Node_AddDelegate}
   n9 = Event PreConstruct()
   > Parent: PreConstruct(IsDesignTime=@n9.IsDesignTime)
   > "SetText (Text)"(self=Get Text_SettingName() -> (Text_SettingName:Common Text Object Reference), InText="LOCTEXT(\"Mode\")")
   > PopulateEntries()
  ]}
 GetSelectedMode {nodes [
   GetSelectedMode() {is=title K2Node_FunctionEntry}
   n1 = GetSelectedIndex!(self:Common Rotator Object Reference=Get Rotator_SettingValue()) -> (ReturnValue:int=0)
   > Return Node!(
    execute
    Mode:EMusicKeyMode Enum=Byte to Enum EMusicKeyMode!(Byte:Byte="To Byte (Integer)"!(InInt:int=@n1) -> (ReturnValue:Byte=0)) -> (ReturnValue:EMusicKeyMode Enum) {is=title K2Node_CastByteToEnum}
   )
  ]}
 OnModeChanged {nodes [OnModeChanged!() -> (then, NewMode:EMusicKeyMode Enum=Major) {is=title K2Node_FunctionEntry}]}
 OverrideModeEntry {nodes [
   n1 = OverrideModeEntry!() -> (then, ModePreview:EMusicKeyMode Enum=Major) {is=title K2Node_FunctionEntry}
   > SetSelectedItem(self=Get Rotator_SettingValue(), InValue="To Integer (Byte)"(InByte=@n1) -> (ReturnValue=0))
  ]}
 PopulateEntries {nodes [
   PopulateEntries() {is=title K2Node_FunctionEntry}
   > n1 = ForEach EMusicKeyMode!(execute:Exec, SkipHidden:bool) -> (LoopBody:Exec, EnumValue:EMusicKeyMode Enum, then:Exec) {is=title K2Node_ForEachElementInEnum}
   PopulateTextLabels!(
    execute=@n1.then
    self:Common Rotator Object Reference=Get Rotator_SettingValue()
    Labels:Array of Texts=Get Texts() -> (Texts:Array of Texts)
   ) -> (then)
   Branch(execute=@n1.LoopBody, Condition="Not Equal (Enum)"!(A:EMusicKeyMode Enum=@n1, B:EMusicKeyMode Enum=Num) -> (ReturnValue:bool))
   > Add Unique(
    TargetArray:Array of Texts=Get Texts()
    NewItem:"Text (by ref)"=Select!(
     Major:"Text (by ref)"=Format Text(Format="LOCTEXT(\"Major\")")
     Minor:"Text (by ref)"=Format Text(Format="LOCTEXT(\"Minor\")")
     Num:"Text (by ref)"
     Index:EMusicKeyMode Enum=@n1
    ) -> (ReturnValue:"Text (by ref)")
   )
  ]}
]
variables [
 FocusAnimation {type=WidgetAnimation, default=null}
 OnModeChanged {}
]
widgets {
 info {
  inheritedWidgetCount=0
  parentClass=&/SettingsListEntry_JamDiscrete.SettingsListEntry_JamDiscrete_C
  rootWidgetClass=null
  widgetCount=0
 }
 widgets []
}
