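# Jam settings list entry that lets the player switch jam privacy between Open and Closed with a rotator.
# Rotator changes go to BP_JamControllerComponent to set the music manager's privacy.
# Loop start/stop and routed PrivacyChanged events refresh the displayed value and availability.
# The control is disabled with an error message in persistent sessions or while non-party players jam (team/squad check in editor/PIE).
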WidgetBlueprint /FMJamSystem/Widgets/GlobalControls/WBP_FMJamPrivacySelector
graphs [
 EventGraph {nodes [
   RefreshPrivacyUI() {is=title K2Node_CustomEvent}
   > n1 = GetJamSessionForActor!(execute, Actor:Actor Object Reference=GetOwningPlayer()) -> (then, ReturnValue:Jam Session Interface Interface)
   > UpdatePrivacyEvent!(
    execute
    self
    PrivacySetting:EJamPrivacySetting Enum=Get CurrentJamPrivacy!(self:Jam Session Component Music Manager Object Reference=GetMusicManager!(self:Jam Session Interface Interface=@n1) -> (ReturnValue:Jam Session Component Music Manager Object Reference)) -> (CurrentJamPrivacy:EJamPrivacySetting Enum=Public)
   ) -> (then)
   > n2 = Branch(Condition=NOT Boolean(A=GetIsPersistent!(self:Jam Session Interface Interface=@n1) -> (ReturnValue:bool)))
   > n3 = IsNonPartyMemberInJam!(execute, self) -> (then, JamHasNonTeammate:bool)
   > n4 = Branch(Condition=@n3)
   > OnErrorStateChange!(execute, self, hasErrorMessage:bool=true, ErrorText:Text=Get OtherAreJammingPrivacyText() -> (OtherAreJammingPrivacyText:Text)) -> (then)
   > OnAvailableChanged!(execute, self, isChangePrivacyAvailable:bool=false) -> (then)
   OnErrorStateChange(execute=@n4.else, hasErrorMessage=false)
   > OnAvailableChanged(isChangePrivacyAvailable=true)
   OnErrorStateChange(execute=@n2.else, hasErrorMessage=true, ErrorText=Get LockedPersistentJamPrivacy() -> (LockedPersistentJamPrivacy:Text))
   > OnAvailableChanged(isChangePrivacyAvailable=false)
   OnButtonIncrease!() -> (OutputDelegate, then, Button:Common Button Legacy Object Reference) {is=title K2Node_CustomEvent}
   > ShiftTextRight(self:Common Rotator Object Reference=Get Rotator_SettingValue() -> (Rotator_SettingValue:Fort Settings Rotator Object Reference))
   OnButtonDecrease!() -> (OutputDelegate, then, Button:Common Button Legacy Object Reference) {is=title K2Node_CustomEvent}
   > ShiftTextLeft(self:Common Rotator Object Reference=Get Rotator_SettingValue())
   n5 = UpdatePrivacyEvent!() -> (OutputDelegate, then, PrivacySetting:EJamPrivacySetting Enum=Public) {is=title K2Node_CustomEvent}
   > SetSelectedItem!(
    execute
    self:Common Rotator Object Reference=Get Rotator_SettingValue()
    InValue:int="To Integer (Byte)"!(InByte:Byte=@n5.PrivacySetting) -> (ReturnValue:int=0)
   ) -> (then)
   n6 = PrivacyRotatorValueChanged!() -> (OutputDelegate, then, Value:int) {is=title K2Node_CustomEvent}
   > n7 = GetJamSessionForActor(Actor=GetOwningPlayer())
   n8 = GetMusicManager(self=@n7)
   n9 = Byte to Enum EJamPrivacySetting!(Byte:Byte="To Byte (Integer)"!(InInt:int=@n6.Value) -> (ReturnValue:Byte=0)) -> (ReturnValue:EJamPrivacySetting Enum) {is=title K2Node_CastByteToEnum}
   > Branch(Condition="Not Equal (Enum)"!(A:EJamPrivacySetting Enum=Get CurrentJamPrivacy(self=@n8) -> (CurrentJamPrivacy=Public), B:EJamPrivacySetting Enum=@n9) -> (ReturnValue:bool))
   n10 = Get Components By Class!(
    self:Actor Object Reference=GetOwningPlayer()
    ComponentClass:Actor Component Class Reference=/FMJamSystem/Components/BP_JamControllerComponent.BP_JamControllerComponent_C
   ) -> (ReturnValue:Array of BP Jam Controller Component Object References)
   > Branch(Condition=Is Valid Index!(TargetArray:Array of BP Jam Controller Component Object References=@n10, IndexToTest:int) -> (ReturnValue:bool))
   > SetCurrentJamPrivacyForMusicManager!(
    execute
    self:BP Jam Controller Component Object Reference="Get (a copy)"(Array:Array of BP Jam Controller Component Object References=@n10) -> (Output:BP Jam Controller Component Object Reference)
    JamPrivacy:EJamPrivacySetting Enum=@n9
   ) -> (then)
   LoopChanged!() -> (OutputDelegate, then, PlayParams:"Jam Play Params Structure (by ref)", bChangedLoop:bool=~) {is=title K2Node_CustomEvent}
   > RefreshPrivacyUI() {is=title K2Node_CallFunction}
   Event Destruct()
   > Parent: Destruct()
   > n11 = GetJamSessionForActor(Actor=GetOwningPlayer())
   n12 = GetMusicManager(self=@n11)
   n13 = Create Event()
   > Unbind Event from On Loop Started(self:Jam Session Component Music Manager Object Reference=@n12, Delegate=@n13) {is=title K2Node_RemoveDelegate}
   > Unbind Event from On Loop Stopped(self:Jam Session Component Music Manager Object Reference=@n12, Delegate=@n13) {is=title K2Node_RemoveDelegate}
   Event Construct()
   > Parent: Construct()
   > Assign On Button Clicked!(
    execute:Exec
    self:Common Button Legacy Object Reference=Get Button_Decrease() -> (Button_Decrease:Common Button Legacy Object Reference)
    Delegate:"Delegate (by ref)"=Create Event()
   ) -> (then:Exec) {is=title K2Node_AssignDelegate}
   > Bind Event to On Button Clicked(
    self:Common Button Legacy Object Reference=Get Button_Increase() -> (Button_Increase:Common Button Legacy Object Reference)
    Delegate=Create Event()
   ) {is=title K2Node_AddDelegate}
   > Bind Event to On Rotated(self:Common Rotator Object Reference=Get Rotator_SettingValue(), Delegate=Create Event()) {is=title K2Node_AddDelegate}
   > n14 = Sequence!(execute) -> (then_0, then_1, then_2:Exec)
   > n15 = GetJamSessionForActor(Actor=GetOwningPlayer())
   n16 = Create Event()
   n17 = GetMusicManager(self=@n15)
   > Bind Event to On Loop Started(self:Jam Session Component Music Manager Object Reference=@n17, Delegate=@n16) {is=title K2Node_AddDelegate}
   > Bind Event to On Loop Stopped(self:Jam Session Component Music Manager Object Reference=@n17, Delegate=@n16) {is=title K2Node_AddDelegate}
   n18 = GetJamSessionForActor(execute=@n14.then_1, Actor=GetOwningPlayer())
   > n19 = StartListeningToEvent!(
    execute:Exec
    Target:Gameplay Event Router Component Object Reference=GetGameplayEventRouterComponent!(self:Jam Session Interface Interface=@n18) -> (ReturnValue:Gameplay Event Router Component Object Reference)
    EventType:Script Struct Object Reference=/Script/FMJamSessionRuntime.JamEvent_PrivacyChanged
    Context:Object Reference
    EventBubblingRule:EEventBubblingRule Enum
   ) -> (
    then:Exec
    AsyncTaskProxy:Async Action Start Listening To Event Object Reference
    OnEventReceived:Exec
    Payload:Jam Event Privacy Changed Structure
   )
   RefreshPrivacyUI(execute=@n19.OnEventReceived)
   RefreshPrivacyUI(execute=@n14.then_2)
   n20 = Event PreConstruct()
   > Parent: PreConstruct(IsDesignTime=@n20.IsDesignTime)
   > Branch(Condition=@n20.IsDesignTime)
   > SetVisibility(
    self=[
     Get ExpansionSlot() -> (ExpansionSlot:Named Slot Object Reference)
     Get SettingExpansion_Spacer() -> (SettingExpansion_Spacer:Spacer Object Reference)
    ]
    InVisibility=SelfHitTestInvisible
   )
   Event OnInitialized()
   > Parent: OnInitialized() {is=title K2Node_CallParentFunction}
   > "SetText (Text)"(self=Get Text_SettingName() -> (Text_SettingName:Common Text Object Reference), InText="LOCTEXT(\"Privacy\")")
   > PopulateEntries()
  ]}
 IsJamOpen {nodes [
   IsJamOpen() {is=title K2Node_FunctionEntry}
   > n1 = Sequence!(execute) -> (then_0, then_1)
   > n2 = GetJamSessionForActor(Actor=GetOwningPlayer())
   > n3 = IsNonPartyMemberInJam()
   n4 = GetMusicManager(self=@n2)
   n5 = Get CurrentJamPrivacy(self=@n4) -> (CurrentJamPrivacy=Public)
   n6 = "Equal (Enum)"(A:EJamPrivacySetting Enum=@n5, B:EJamPrivacySetting Enum=PartyOnly)
   > Set OpenToPlayer(OpenToPlayer:bool=NOT Boolean(A=AND Boolean(A=@n3, B=@n6))) -> (Output_Get:bool)
   > Set OpenToOthers(OpenToOthers:bool="Equal (Enum)"(A=@n5, B=Public)) -> (Output_Get:bool)
   Return Node!(
    execute=@n1.then_1
    isJamOpenToPlayer:bool=Get OpenToPlayer() -> (OpenToPlayer:bool)
    isJamOpenToOthers:bool=Get OpenToOthers() -> (OpenToOthers:bool)
   )
   Get CurrentJamPrivacy(self=@n4) -> (CurrentJamPrivacy=Public)
  ]}
 IsNonPartyMemberInJam {nodes [
   IsNonPartyMemberInJam() {is=title K2Node_FunctionEntry}
   > n1 = Sequence()
   > n2 = GetJamSessionForActor(Actor=GetOwningPlayer())
   n3 = GetMusicSlotsInUse!(self:Jam Session Component Music Manager Object Reference=GetMusicManager(self=@n2)) -> (ReturnValue:Array of Jam Music Slot Object References)
   > n4 = Branch(Condition=Is Not Empty!(TargetArray:Array of Jam Music Slot Object References=@n3) -> (ReturnValue:bool))
   > n5 = For Each Loop with Break!(Exec:Exec, Array:Array of Jam Music Slot Object References=@n3, Break:Exec=@n12.then) -> (LoopBody:Exec, Array Element:Jam Music Slot Object Reference, Array Index:int, Completed:Exec)
   > n6 = Cast To FortPlayerPawnAthena(Object=GetPawnFromActor!(Actor:Actor Object Reference=GetOwningPlayer(), PawnType:Pawn Class Reference) -> (ReturnValue:Pawn Object Reference))
   n7 = Break Jam Play Params!(JamPlayParams:"Jam Play Params Structure (by ref)"=GetPlayParams!(self:Jam Music Slot Object Reference=@n5) -> (ReturnValue:"Jam Play Params Structure (by ref)")) -> (
    SongShortName:Name
    LoopType:EFMJamLoopType Enum
    PlayerNetId:Unique Net Id Structure
    LoopInstanceId:int
    bIsAutoJammer:bool=~
    PlayingActorOverride:Actor Weak object
    ServerQuantizationTimeSeconds:float=~
    SlotStartMethod:EJamSlotStartMethod Enum
   ) {is=title K2Node_BreakStruct}
   > n8 = Cast To FortPlayerPawnAthena(Object=GetPawnFromActor(Actor=GetPlayerStateFromUniqueNetId!(UniqueId:"Unique Net Id Structure (by ref)"=@n7.PlayerNetId) -> (ReturnValue:Player State Object Reference)))
   n9 = IsPlayerInLocalParty!(OtherPlayerPawn:Fort Player Pawn Athena Object Reference=@n8) -> (ReturnValue:bool)
   n10 = InEditorOrPIE!() -> (ReturnValue:bool)
   n11 = OR Boolean(
    A=OnSameTeam!(ActorA:Actor Object Reference=@n8, ActorB:Actor Object Reference=@n6) -> (ReturnValue:bool)
    B=OnSameSquad!(PlayerPawnA:Fort Player Pawn Athena Object Reference=@n8, PlayerPawnB:Fort Player Pawn Athena Object Reference=@n6) -> (ReturnValue:bool)
   )
   > Branch(Condition=NOT Boolean(A=Select(Option 0:bool=@n9, Option 1:bool=@n11, Index=@n10) -> (ReturnValue:bool)))
   > n12 = Set NonTeammateFound(NonTeammateFound:bool=true) -> (Output_Get:bool)
   Set NonTeammateFound(execute=@n4.else, NonTeammateFound=false)
   Return Node!(execute=@n1.then_1, JamHasNonTeammate:bool=Get NonTeammateFound() -> (NonTeammateFound:bool))
  ]}
 OnAvailableChanged {nodes [
   n1 = OnAvailableChanged!() -> (then, isChangePrivacyAvailable:bool=~) {is=title K2Node_FunctionEntry}
   > SetIsEnabled(bInIsEnabled=@n1)
   > SetIsEnabledVisual!(execute, self:Settings List Entry Jam Discrete Object Reference, IsEnabled:bool=@n1) -> (then)
   > SetIsEnabled(self=Get Rotator_SettingValue(), bInIsEnabled=@n1)
   > SetVisibility(
    self=Get Overlay_Value() -> (Overlay_Value:Overlay Object Reference)
    InVisibility=Select(Option 0:ESlateVisibility Enum=Collapsed, Option 1:ESlateVisibility Enum=SelfHitTestInvisible, Index=@n1) -> (ReturnValue:ESlateVisibility Enum)
   )
  ]}
 OnErrorStateChange {nodes [
   n1 = OnErrorStateChange!() -> (then, hasErrorMessage:bool=~, ErrorText:Text) {is=title K2Node_FunctionEntry}
   > n2 = Sequence()
   > Branch(Condition=@n1)
   > "SetText (Text)"(self=Get MessageTextBlock() -> (MessageTextBlock:Text Block Object Reference), InText=@n1.ErrorText)
   SetVisibility(
    execute=@n2.then_1
    self=[Get ExpansionSlot(), Get SettingExpansion_Spacer()]
    InVisibility=Select(Option 0=Collapsed, Option 1=SelfHitTestInvisible, Index=@n1)
   )
  ]}
 PopulateEntries {nodes [
   PopulateEntries() {is=title K2Node_FunctionEntry}
   > n1 = ForEach EJamPrivacySetting!(execute:Exec, SkipHidden:bool) -> (LoopBody:Exec, EnumValue:EJamPrivacySetting Enum, then:Exec) {is=title K2Node_ForEachElementInEnum}
   PopulateTextLabels!(
    execute=@n1.then
    self:Common Rotator Object Reference=Get Rotator_SettingValue()
    Labels:Array of Texts=Get Texts() -> (Texts:Array of Texts)
   ) -> (then)
   Branch(execute=@n1.LoopBody, Condition="Not Equal (Enum)"(A=@n1, B=Num))
   > Add Unique(
    TargetArray:Array of Texts=Get Texts()
    NewItem:"Text (by ref)"=Select!(
     Public:"Text (by ref)"=Format Text(Format="LOCTEXT(\"Open\")")
     PartyOnly:"Text (by ref)"=Format Text(Format="LOCTEXT(\"Closed\")")
     Num:"Text (by ref)"
     Index:EJamPrivacySetting Enum=@n1
    ) -> (ReturnValue:"Text (by ref)")
   )
  ]}
]
variables [
 LockedPersistentJamPrivacy {type=string, default=Privacy can't be changed.}
 OtherAreJammingPrivacyText {type=string, default=Cannot change while non-squad players are Jamming}
 PrivacyOptions {type="array<string>", default []}
]
widgets {
 info {
  inheritedWidgetCount=0
  namedSlotCount=1
  parentClass=&/SettingsListEntry_JamDiscrete.SettingsListEntry_JamDiscrete_C
  rootWidgetClass=null
  widgetCount=5
 }
 widgets [
  Overlay_Expansion {widgetClassPath=&Overlay, widget=Overlay_Expansion}
  Image_Background {
   widgetClassPath=&Image
   parent=Overlay_Expansion
   widget=Image_Background
   properties {
    brush {
     bIsDynamicallyLoaded=false
     drawAs=Image
     imageType=FullColor
     mirroring=NoMirror
     resourceName=None
     resourceObject=/FortUILibrary/Systems/InputField/MaterialInstances/MI_UIKit_BackgroundRounded
     tiling=NoTile
     imageSize=<32 32>
     margin=<ltrb 0 0 0 0>
     outlineSettings {
      bUseBrushTransparency=false
      roundingType=HalfHeightRadius
      width=0
      color {colorUseRule=UseColor_Specified, specifiedColor=<rgba 0 0 0 0>}
      cornerRadii=<0 0 0 0>
     }
     tintColor {colorUseRule=UseColor_Specified, specifiedColor=<rgba 1 1 1 1>}
     uVRegion {bIsValid=false, max=<0 0>, min=<0 0>}
    }
    colorAndOpacity=<rgba 1 0 0 1>
   }
   slot {class=OverlaySlot, properties {horizontalAlignment=HAlign_Fill, verticalAlignment=VAlign_Fill}}
  }
  HorizontalBox_ExpansionMessage {
   widgetClassPath=&HorizontalBox
   parent=Overlay_Expansion
   widget=HorizontalBox_ExpansionMessage
   slot {class=OverlaySlot, properties {horizontalAlignment=HAlign_Fill, verticalAlignment=VAlign_Fill, padding=<ltrb 5 0 5 0>}}
  }
  Iconplaceholder {
   widgetClassPath=&Image
   bIsVariable=true
   parent=HorizontalBox_ExpansionMessage
   widget=Iconplaceholder
   properties {brush {
     bIsDynamicallyLoaded=false
     drawAs=Image
     imageType=NoImage
     mirroring=NoMirror
     resourceName=None
     resourceObject=/Game/UI/Foundation/Textures/Buttons/T-SlotIcon-NotRecyclable-64
     tiling=NoTile
     imageSize=<32 32>
     margin=<ltrb 0 0 0 0>
     outlineSettings {
      bUseBrushTransparency=false
      roundingType=HalfHeightRadius
      width=0
      color {colorUseRule=UseColor_Specified, specifiedColor=<rgba 0 0 0 0>}
      cornerRadii=<0 0 0 0>
     }
     tintColor {colorUseRule=UseColor_Specified, specifiedColor=<rgba 1 0.5089 0.5089 1>}
     uVRegion {bIsValid=false, max=<0 0>, min=<0 0>}
    }}
   slot {class=HorizontalBoxSlot, properties {horizontalAlignment=HAlign_Right, verticalAlignment=VAlign_Center}}
  }
  MessageTextBlock {
   widgetClassPath=&UEFN_TextBlock
   bIsVariable=true
   parent=HorizontalBox_ExpansionMessage
   widget=MessageTextBlock
   properties {
    autoWrapText=true
    text=Cannot change while non-squad players are Jamming
    colorAndOpacity {colorUseRule=UseColor_Specified, specifiedColor=<rgba 1 0.5104 0.5104 1>}
    font {
     bForceMonospaced=false
     bMaterialIsStencil=false
     fontMaterial=null
     fontObject=&HeadingNow
     letterSpacing=0
     monospacedWidth=1
     size=13.5
     skewAmount=0
     typefaceFontName=74 Regular
     outlineSettings {
      bApplyOutlineToDropShadows=false
      bMiteredCorners=false
      bSeparateFillAlpha=false
      outlineMaterial=null
      outlineSize=0
      outlineColor=<rgba 0 0 0 1>
     }
    }
   }
   slot {class=HorizontalBoxSlot, properties {verticalAlignment=VAlign_Center, padding=<ltrb 3 4 0 4>, size {sizeRule=Fill, value=1}}}
  }
 ]
}
