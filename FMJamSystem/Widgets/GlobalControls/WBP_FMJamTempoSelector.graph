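# Tempo selector widget for the Jam system, a settings-list rotator over a min-to-max BPM range that broadcasts changes through OnTempoChanged.
# Refresh seeds it from the jam session's music manager, falling back to 120, and skips updates during button or controller cycling.
# Increase/decrease buttons use timers to step once on tap and repeat on hold, disabling at the hard-coded index limits.
# Rotations jumping over one step revert, blocking wraparound.
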WidgetBlueprint /FMJamSystem/Widgets/GlobalControls/WBP_FMJamTempoSelector
graphs [
 EventGraph {nodes [
   n1 = "On Rotated (Rotator_SettingValue)"!() -> (OutputDelegate, then, Value:int=0) {is=title K2Node_ComponentBoundEvent}
   > n2 = Sequence!(execute) -> (then_0, then_1, then_2:Exec)
   > n3 = Get!(execute:Exec) -> (then:Exec, else:Exec, CurrentHoveredButton:Common Button Legacy Object Reference)
   Set IsCyclingWithController(execute=@n3.else, IsCyclingWithController:bool=true) -> (Output_Get:bool)
   > RetriggerableDelay(Duration=0.25)
   > Set IsCyclingWithController(IsCyclingWithController=false)
   RefreshExtents(execute=@n2.then_2)
   n4 = int - int(A=Get lastValue() -> (lastValue:int=0), B=@n1.Value)
   n5 = Branch(execute=@n2.then_1, Condition="integer > integer"(A="Absolute (Integer)"!(A:int=@n4) -> (ReturnValue:int=0), B=1))
   > SetSelectedItem!(
    execute
    self:Common Rotator Object Reference=Get Rotator_SettingValue() -> (Rotator_SettingValue:Fort Settings Rotator Object Reference)
    InValue:int=Get lastValue() -> (lastValue=0)
   ) -> (then)
   Set lastValue(execute=@n5.else, lastValue:int=@n1.Value) -> (Output_Get:int=0)
   > Call On Tempo Changed!(
    execute:Exec
    self:WBP FMJam Tempo Selector Object Reference
    NewTempo:int=int + int(A=Get lastValue() -> (lastValue=0), B=Get minTempo() -> (minTempo:int=0))
   ) -> (then:Exec) {is=title K2Node_CallDelegate}
   Refresh() {is=title K2Node_CustomEvent}
   > n6 = GetJamSessionForActor!(execute, Actor:Actor Object Reference=GetOwningPlayer()) -> (then, ReturnValue:Jam Session Interface Interface)
   > n7 = Get()
   n8 = Branch(execute=@n7.else, Condition=Get IsCyclingWithController() -> (IsCyclingWithController:bool))
   n9 = TryGetKeyModeTempo!(self:Jam Session Component Music Manager Object Reference=GetMusicManager!(self:Jam Session Interface Interface=@n6) -> (ReturnValue:Jam Session Component Music Manager Object Reference)) -> (OutMusicKey:EMusicKey Enum=C, OutMusicKeyMode:EMusicKeyMode Enum=Major, OutTempo:int=0, ReturnValue:bool)
   n10 = Branch(execute=@n8.else, Condition=AND Boolean(A=@n9.ReturnValue, B="Not Equal (Integer)"(A=@n9.OutTempo)))
   > n11 = Set lastValue(lastValue=int - int(A=@n9.OutTempo, B=Get minTempo() -> (minTempo=0))) -> (Output_Get=0)
   > SetSelectedItem(self=Get Rotator_SettingValue(), InValue=@n11)
   SetSelectedItem(execute=@n10.else, self=Get Rotator_SettingValue(), InValue=int - int(A=120, B=Get minTempo() -> (minTempo=0)))
   "OnButtonUnhovered(Button)"!() -> (OutputDelegate, then, Button:Common Button Legacy Object Reference) {is=title K2Node_CustomEvent}
   > Set CurrentHoveredButton(CurrentHoveredButton:Common Button Legacy Object Reference) -> (Output_Get:Common Button Legacy Object Reference)
   > ClearTimers()
   n12 = CheckForPress_Event() {is=title K2Node_CustomEvent}
   n13 = CheckForHold_Event() {is=title K2Node_CustomEvent}
   n14 = Set Timer by Event(execute=@n12.then, Delegate=@n13, Time=0.03, bLooping=true, bMaxOncePerFrame=true)
   > Set CheckForHoldTimer(CheckForHoldTimer:Timer Handle Structure=@n14) -> (Output_Get:Timer Handle Structure)
   ClearTimers() {is=title K2Node_CustomEvent}
   > Clear and Invalidate Timer by Handle(Handle=Get CheckForHoldTimer() -> (CheckForHoldTimer:Timer Handle Structure))
   > Clear and Invalidate Timer by Handle(Handle=Get CheckForPressTimer() -> (CheckForPressTimer:Timer Handle Structure))
   DecrementIfAllowed() {is=title K2Node_CustomEvent}
   n15 = "Equal (Integer)"(A=GetSelectedIndex!(self:Common Rotator Object Reference=Get Rotator_SettingValue()) -> (ReturnValue:int=0))
   > n16 = Branch(Condition=@n15)
   ShiftTextLeft(execute=@n16.else, self:Common Rotator Object Reference=Get Rotator_SettingValue())
   > Branch(Condition=@n15)
   > ClearTimers() {is=title K2Node_CallFunction}
   IncrementIfAllowed() {is=title K2Node_CustomEvent}
   n17 = "Equal (Integer)"(A=GetSelectedIndex(self=Get Rotator_SettingValue()) -> (ReturnValue=0), B=100)
   > n18 = Branch(Condition=@n17)
   ShiftTextRight(execute=@n18.else, self:Common Rotator Object Reference=Get Rotator_SettingValue())
   > Branch(Condition=@n17)
   > ClearTimers()
   n19 = Debug Buttons State!() -> (OutputDelegate, then, NewParam:double=~) {is=title K2Node_CustomEvent}
   n20 = Get Button_Increase() -> (Button_Increase:Common Button Legacy Object Reference)
   n21 = IsPressed!(self:Common Button Base Object Reference=@n20) -> (ReturnValue:bool)
   n22 = IsHovered(self=@n20)
   n23 = IsInteractionEnabled!(self:Common Button Base Object Reference=@n20) -> (ReturnValue:bool)
   > PrintString!(
    execute
    InString=Append!(
     A="Increase ( P "
     B="To String (Boolean)"(InBool=@n21)
     C:String=" H "
     D:String="To String (Boolean)"(InBool=@n22)
     E:String=" E "
     F:String="To String (Boolean)"(InBool=@n23)
     G:String=")"
    ) -> (ReturnValue)
    bPrintToScreen
    bPrintToLog=false
    TextColor
    Duration=@n19.NewParam
    Key=0
   ) -> (then)
   n24 = Get Button_Decrease() -> (Button_Decrease:Common Button Legacy Object Reference)
   n25 = IsInteractionEnabled(self=@n24)
   n26 = IsHovered(self=@n24)
   n27 = IsPressed(self=@n24)
   > PrintString(
    InString=Append(
     A="Decrease ( P "
     B="To String (Boolean)"(InBool=@n27)
     C=" H "
     D="To String (Boolean)"(InBool=@n26)
     E=" E "
     F="To String (Boolean)"(InBool=@n25)
     G=")"
    )
    bPrintToLog=false
    Duration=@n19.NewParam
    Key=1
   )
   OnButtonReleased!() -> (OutputDelegate, then, Button:Common Button Legacy Object Reference) {is=title K2Node_CustomEvent}
   > n28 = Sequence!(execute) -> (then_0, then_1)
   > Branch(Condition=Is Timer Active by Handle!(Handle:Timer Handle Structure=Get CheckForPressTimer()) -> (ReturnValue:bool))
   > n29 = Branch(Condition="Equal (Object)"(A=Get CurrentHoveredButton() -> (CurrentHoveredButton:Common Button Legacy Object Reference), B=Get Button_Increase()))
   > IncrementIfAllowed() {is=title K2Node_CallFunction}
   DecrementIfAllowed(execute=@n29.else) {is=title K2Node_CallFunction}
   ClearTimers(execute=@n28.then_1)
   n30 = OnButtonPressed!() -> (OutputDelegate, then, Button:Common Button Legacy Object Reference) {is=title K2Node_CustomEvent}
   > Set CurrentHoveredButton(CurrentHoveredButton=@n30.Button)
   > n31 = Set Timer by Event(Delegate=@n12, Time=0.25, bLooping=false)
   > Set CheckForPressTimer(CheckForPressTimer:Timer Handle Structure=@n31) -> (Output_Get:Timer Handle Structure)
   RefreshExtents() {is=title K2Node_CustomEvent}
   > n32 = Sequence()
   n33 = Get Rotator_SettingValue()
   > n34 = Branch(Condition="Equal (Integer)"(A=GetSelectedIndex(self=@n33) -> (ReturnValue=0)))
   > DisableButtonWithReason!(
    execute
    self:Common Button Base Object Reference=Get Button_Decrease()
    DisabledReason:"Text (by ref)"=MakeLiteralText(Value="LOCTEXT(\"valueLimit\")")
   ) -> (then)
   > Set CurrentHoveredButton()
   > ClearTimers()
   SetIsEnabled(execute=@n34.else, self=Get Button_Decrease(), bInIsEnabled=true)
   n35 = Get Rotator_SettingValue()
   n36 = Branch(execute=@n32.then_1, Condition="Equal (Integer)"(A=GetSelectedIndex(self=@n35) -> (ReturnValue=0), B=100))
   > DisableButtonWithReason(self=Get Button_Increase(), DisabledReason=MakeLiteralText(Value="LOCTEXT(\"valueLimit\")"))
   > Set CurrentHoveredButton()
   > ClearTimers()
   SetIsEnabled(execute=@n36.else, self=Get Button_Increase(), bInIsEnabled=true)
   n37 = Event OnAddedToFocusPath!() -> (OutputDelegate, then, InFocusEvent:Focus Event Structure) {is=title K2Node_Event}
   > Parent: OnAddedToFocusPath!(execute, InFocusEvent:Focus Event Structure=@n37.InFocusEvent) -> (then) {is=title K2Node_CallParentFunction}
   n38 = Get FocusAnimation() -> (FocusAnimation:Widget Animation Object Reference)
   > Is Valid(InputObject=@n38)
   > PlayAnimationForward(InAnimation=@n38)
   n39 = Event OnRemovedFromFocusPath!() -> (OutputDelegate, then, InFocusEvent:Focus Event Structure) {is=title K2Node_Event}
   > Parent: OnRemovedFromFocusPath!(execute, InFocusEvent:Focus Event Structure=@n39.InFocusEvent) -> (then) {is=title K2Node_CallParentFunction}
   n40 = Get FocusAnimation()
   > Is Valid(InputObject=@n40)
   > PlayAnimationReverse(InAnimation=@n40)
   Event Construct()
   > Parent: Construct()
   > n41 = Cast To WBP_Jam_GC_IconTextButton!(execute, Object=Get Button_Decrease()) -> (then, CastFailed, AsWBP Jam GC Icon Text Button:WBP Jam GC Icon Text Button Object Reference) {is=title K2Node_DynamicCast}
   > Bind Event to On Button Released(self:WBP Jam GC Icon Text Button Object Reference=@n41, Delegate=Create Event()) {is=title K2Node_AddDelegate}
   > Bind Event to On Button Pressed(self:WBP Jam GC Icon Text Button Object Reference=@n41, Delegate=Create Event()) {is=title K2Node_AddDelegate}
   > Bind Event to On Button Unhovered(self:Common Button Legacy Object Reference=@n41, Delegate=Create Event()) {is=title K2Node_AddDelegate}
   > n42 = Cast To WBP_Jam_GC_IconTextButton(Object=Get Button_Increase())
   > Bind Event to On Button Released(self=@n42, Delegate=Create Event())
   > Bind Event to On Button Pressed(self=@n42, Delegate=Create Event())
   > Bind Event to On Button Unhovered(self=@n42, Delegate=Create Event())
   > Refresh() {is=title K2Node_CallFunction}
   n43 = Event PreConstruct()
   > Parent: PreConstruct(IsDesignTime=@n43.IsDesignTime)
   > "SetText (Text)"(self=Get Text_SettingName() -> (Text_SettingName:Common Text Object Reference), InText="LOCTEXT(\"Tempo\")")
   > PopulateEntries()
   IncrementIfAllowed(execute=@n47.then)
   DecrementIfAllowed(execute=@n47.else)
   ClearTimers(execute=[@n48.else, @n49.Is Not Valid])
   n44 = Get CurrentHoveredButton()
   n45 = IsPressed(self=@n44)
   n46 = AND Boolean(A=@n45, B=Get bIsEnabled!(self:Widget Object Reference=@n44) -> (bIsEnabled:bool))
   n47 = Branch(execute=@n48.then, Condition="Equal (Object)"(A=Get CurrentHoveredButton(), B=Get Button_Increase()))
   n48 = Branch(execute=@n49.Is Valid, Condition=@n46)
   n49 = Is Valid(exec=@n13.then, InputObject=@n44)
  ]}
 GetSelectedKey {nodes [
   GetSelectedKey() {is=title K2Node_FunctionEntry}
   n1 = GetSelectedIndex(self=Get Rotator_SettingValue()) -> (ReturnValue=0)
   > Return Node!(
    execute
    Key:EMusicKey Enum=Byte to Enum EMusicKey!(Byte:Byte="To Byte (Integer)"!(InInt:int=@n1) -> (ReturnValue:Byte=0)) -> (ReturnValue:EMusicKey Enum) {is=title K2Node_CastByteToEnum}
   )
  ]}
 OnTempoChanged {nodes [OnTempoChanged!() -> (then, NewTempo:int) {is=title K2Node_FunctionEntry}]}
 OverrideTempoEntry {nodes [
   n1 = OverrideTempoEntry!() -> (then, TempoPreview:int) {is=title K2Node_FunctionEntry}
   > Clear(TargetArray:Array of Texts=Get Texts() -> (Texts:Array of Texts))
   > Add Unique(TargetArray:Array of Texts=Get Texts(), NewItem:"Text (by ref)"="To Text (Integer)"(Value=@n1))
   > PopulateTextLabels!(execute, self:Common Rotator Object Reference=Get Rotator_SettingValue(), Labels:Array of Texts=Get Texts()) -> (then)
   > SetSelectedItem(self=Get Rotator_SettingValue(), InValue=0)
  ]}
 PopulateEntries {nodes [
   PopulateEntries() {is=title K2Node_FunctionEntry}
   > n1 = For Loop(FirstIndex=Get minTempo() -> (minTempo=0), LastIndex=Get maxTempo() -> (maxTempo:int=0))
   PopulateTextLabels(execute=@n1.Completed, self=Get Rotator_SettingValue(), Labels=Get Texts())
   Add Unique(execute=@n1.LoopBody, TargetArray=Get Texts(), NewItem="To Text (Integer)"(Value=@n1))
  ]}
]
variables [
 CheckForHoldTimer {type=TimerHandle}
 CheckForPressTimer {type=TimerHandle}
 Counter {type=integer}
 CurrentHoveredButton {type=CommonButtonLegacy, default=null}
 FocusAnimation {type=WidgetAnimation, default=null}
 IsCyclingWithController {type=boolean}
 lastValue {type=integer}
 maxTempo {type=integer, default=180}
 minTempo {type=integer, default=80}
 OnTempoChanged {}
]
widgets {
 info {
  inheritedWidgetCount=0
  parentClass=&/SettingsListEntry_JamDiscrete.SettingsListEntry_JamDiscrete_C
  rootWidgetClass=null
  widgetCount=0
 }
 widgets []
}
