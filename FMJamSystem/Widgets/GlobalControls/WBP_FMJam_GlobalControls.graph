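# Jam global controls panel for changing the session's tempo, key and mode, or snapping them to a song's catalog values.
# Changes trigger on the audio state and go to the server through JamControllerComponent; key and mode first wait on a settle timer.
# Stateful audio events refresh the selectors; Snap To Vocals availability follows the Lead loop's song.
# Activation broadcasts display state and hides the reticle; a hide-pawn tag closes the panel.
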WidgetBlueprint /FMJamSystem/Widgets/GlobalControls/WBP_FMJam_GlobalControls
graphs [
 BP_GetDesiredFocusTarget {nodes [
   Get Desired Focus Target()
   > Return Node!(execute, ReturnValue:Widget Object Reference=Get TempoSelector_Module() -> (TempoSelector_Module:WBP FMJam Tempo Selector Object Reference))
  ]}
 ChangeKey {nodes [
   n1 = ChangeKey!() -> (then, NewKey:EMusicKey Enum=C) {is=title K2Node_FunctionEntry}
   > n2 = Get!(execute:Exec) -> (then:Exec, else:Exec, JamAudioState:Jam Session Component Audio State Object Reference)
   Branch(
    execute=[@n4.then, @n2.then]
    Condition="Not Equal (Enum)"!(
     A:EMusicKey Enum=Get CurrentKey!(self:Jam Session Component Audio State Object Reference=@n2) -> (CurrentKey:EMusicKey Enum=C)
     B:EMusicKey Enum=@n1
    ) -> (ReturnValue:bool)
   )
   > TriggerKeyChange!(execute, self:Jam Session Component Audio State Object Reference=@n2, InKey:EMusicKey Enum=@n1) -> (then)
   > Server_SetKey!(
    execute
    self:Jam Controller Component Object Reference=Get JamControllerComponent() -> (JamControllerComponent:Jam Controller Component Object Reference)
    Key:EMusicKey Enum=@n1
   ) -> (then)
   n3 = InitializeVariables!(execute=@n2.else, self) -> (then, Initialized:bool)
   > n4 = Branch(Condition=@n3)
  ]}
 ChangeMode {nodes [
   n1 = ChangeMode!() -> (then, NewMode:EMusicKeyMode Enum=Major) {is=title K2Node_FunctionEntry}
   > n2 = Get()
   Branch(
    execute=[@n4.then, @n2.then]
    Condition="Not Equal (Enum)"(
     A:EMusicKeyMode Enum=Get CurrentMode!(self:Jam Session Component Audio State Object Reference=@n2) -> (CurrentMode:EMusicKeyMode Enum=Major)
     B:EMusicKeyMode Enum=@n1
    )
   )
   > TriggerModeChange!(execute, self:Jam Session Component Audio State Object Reference=@n2, InMode:EMusicKeyMode Enum=@n1) -> (then)
   > Server_SetMode!(execute, self:Jam Controller Component Object Reference=Get JamControllerComponent(), Mode:EMusicKeyMode Enum=@n1) -> (then)
   n3 = InitializeVariables(execute=@n2.else)
   > n4 = Branch(Condition=@n3)
  ]}
 ChangeTempo {nodes [
   n1 = ChangeTempo!() -> (then, NewTempo:int) {is=title K2Node_FunctionEntry}
   > n2 = Get()
   Branch(
    execute=[@n4.then, @n2.then]
    Condition="Not Equal (Integer)"(A=Get CurrentTempo!(self:Jam Session Component Audio State Object Reference=@n2) -> (CurrentTempo:int=0), B=@n1)
   )
   > TriggerTempoChange!(execute, self:Jam Session Component Audio State Object Reference=@n2, Bpm:int=@n1) -> (then)
   > Server_SetTempo!(execute, self:Jam Controller Component Object Reference=Get JamControllerComponent(), NewBPM:int=@n1) -> (then)
   n3 = InitializeVariables(execute=@n2.else)
   > n4 = Branch(Condition=@n3)
  ]}
 CleanUpMatchVocalsButton {nodes [
   CleanUpMatchVocalsButton() {is=title K2Node_FunctionEntry}
   > n1 = For Each Loop(Array:Array of Jam Music Slot VM Object References=Get MusicSlots!(self:Jam Music Slot Manager VM Object Reference=Get JamMusicSlotManagerVM() -> (JamMusicSlotManagerVM:Jam Music Slot Manager VM Object Reference)) -> (MusicSlots:Array of Jam Music Slot VM Object References)) -> (Array Element:Jam Music Slot VM Object Reference)
   > Remove Field Value Changed Delegate!(
    execute
    self:MVVM Base Viewmodel Object Reference=@n1
    FieldId:Field Notification Id Structure=(FieldName="LoopType")
    Delegate:Delegate=Create Event()
   ) -> (then)
  ]}
 EventGraph {nodes [
   n1 = "On Tempo Changed (TempoSelector_Module)"!() -> (OutputDelegate, then, NewTempo:int=0) {is=title K2Node_ComponentBoundEvent}
   > ChangeTempo!(execute, self, NewTempo:int=@n1.NewTempo) -> (then) {is=title K2Node_CallFunction}
   "On Clicked (SetFromLoop_Button)"!() -> (OutputDelegate, then, Button:Common Button Base Object Reference) {is=title K2Node_ComponentBoundEvent}
   > MatchVocals()
   n2 = "On Mode Changed (ModeSelector_Module)"!() -> (OutputDelegate, then, NewMode:EMusicKeyMode Enum=Major) {is=title K2Node_ComponentBoundEvent}
   > n3 = Set New Mode(New Mode:EMusicKeyMode Enum=@n2.NewMode) -> (Output_Get:EMusicKeyMode Enum=Major)
   n4 = ModeChangeReady() {is=title K2Node_CustomEvent}
   n5 = Set Timer by Event(
    execute=@n3.then
    Delegate=@n4
    Time=GetValueAtLevel!(Input:"Scalable Float Structure (by ref)"=Get SettleBeforeCountTime() -> (SettleBeforeCountTime:Scalable Float Structure), Level:float=0) -> (ReturnValue:float)
    bLooping=false
   )
   > Set ModeChangeTimer(ModeChangeTimer:Timer Handle Structure=@n5) -> (Output_Get:Timer Handle Structure)
   n6 = Get ModeSelectorTimer_Image() -> (ModeSelectorTimer_Image:Image Object Reference)
   > Set Visibility(Visibility=SelfHitTestInvisible, self=@n6)
   n7 = "On Key Changed (DEACTIVATEDKeySelector_Module)"!() -> (OutputDelegate, then, NewKey:EMusicKey Enum=C) {is=title K2Node_ComponentBoundEvent}
   > n8 = Set New Key(New Key:EMusicKey Enum=@n7.NewKey) -> (Output_Get:EMusicKey Enum=C)
   n9 = KeyChangeReady() {is=title K2Node_CustomEvent}
   n10 = Set Timer by Event(execute=@n8.then, Delegate=@n9, Time=GetValueAtLevel(Input=Get SettleBeforeCountTime(), Level=0), bLooping=false)
   > Set KeyChangeTimer(KeyChangeTimer:Timer Handle Structure=@n10) -> (Output_Get:Timer Handle Structure)
   n11 = Get KeySelectorTimer_Image() -> (KeySelectorTimer_Image:Image Object Reference)
   > Set Visibility(Visibility=SelfHitTestInvisible, self=@n11)
   n12 = "On Item Clicked (CommonListView_53)"!() -> (OutputDelegate, then, Item:Object Reference) {is=title K2Node_ComponentBoundEvent}
   > n13 = Set SongSelected(SongSelected:Object Reference=@n12.Item) -> (Output_Get:Object Reference)
   > SetLoopPreview!(execute, self, SongSelected:Object Reference=@n13) -> (then)
   > SwitchLoopTool!(execute, self, Index:int=2) -> (then)
   "On Clicked (ConfirmLoop_Button)"!() -> (OutputDelegate, then, Button:Common Button Base Object Reference) {is=title K2Node_ComponentBoundEvent}
   > n14 = Cast To JamMusicSlotVM!(execute, Object=Get SongSelected() -> (SongSelected:Object Reference)) -> (then, CastFailed, AsJam Music Slot VM:Jam Music Slot VM Object Reference) {is=title K2Node_DynamicCast}
   > SetTempoAndKeyFromShortName!(execute, self, InShortName:Name=Get SongShortName!(self:Jam Music Slot VM Object Reference=@n14) -> (SongShortName:Name=None)) -> (then)
   > Close(self:Jam Global Controls Widget Object Reference)
   PrintString!(execute=@n14.CastFailed, InString="CAST FAILED!", bPrintToScreen, bPrintToLog, TextColor=(R=1,G=0,B=0,A=1), Duration=6, Key) -> (then)
   "On Clicked (Button_Close)"!() -> (OutputDelegate, then, Button:Common Button Base Object Reference) {is=title K2Node_ComponentBoundEvent}
   > Close()
   OnTempoChanged!() -> (OutputDelegate, then, AudioEventTempoChanged:Sparks Core Audio Event Tempo Changed Structure) {is=title K2Node_CustomEvent}
   > Refresh(self:WBP FMJam Tempo Selector Object Reference=Get TempoSelector_Module())
   OnKeyChanged!() -> (OutputDelegate, then, AudioEventKeyChanged:Sparks Core Audio Event Key Changed Structure) {is=title K2Node_CustomEvent}
   > Refresh(self:WBP FMJam Key Selector Object Reference=Get DEACTIVATEDKeySelector_Module() -> (DEACTIVATEDKeySelector_Module:WBP FMJam Key Selector Object Reference))
   OnModeChanged!() -> (OutputDelegate, then, AudioEventKeyModeChanged:Sparks Core Audio Event Key Mode Changed Structure) {is=title K2Node_CustomEvent}
   > Refresh(self:WBP FMJam Mode Selector Object Reference=Get ModeSelector_Module() -> (ModeSelector_Module:WBP FMJam Mode Selector Object Reference))
   StopListeningForTempoKeyModeChanges() {is=title K2Node_CustomEvent}
   > StopListeningToStatefulEvent(self:Async Action Start Listening To Stateful Event Object Reference=Get Tempo Changed Event Handle() -> (Tempo Changed Event Handle:Async Action Start Listening To Stateful Event Object Reference))
   > StopListeningToStatefulEvent(self=Get Key Changed Event Handle() -> (Key Changed Event Handle:Async Action Start Listening To Stateful Event Object Reference))
   > StopListeningToStatefulEvent(self=Get Mode Changed Event Handle() -> (Mode Changed Event Handle:Async Action Start Listening To Stateful Event Object Reference))
   ListenForTempoKeyModeChanges() {is=title K2Node_CustomEvent}
   > n15 = GetJamSession!(execute, self:Jam Global Controls Widget Object Reference) -> (then, ReturnValue:Jam Session Interface Interface)
   > n16 = GetGameplayEventRouterComponent!(execute, self:Jam Session Interface Interface=@n15) -> (then, ReturnValue:Gameplay Event Router Component Object Reference)
   > Branch(Condition=IsValidInterface!(Interface:"Interface Interface (by ref)"=@n15) -> (ReturnValue:bool))
   > Is Valid(InputObject=@n16)
   > n17 = Sequence!(execute) -> (then_0, then_1, then_2:Exec, then_3:Exec)
   > n18 = StartListeningToStatefulEvent!(
    execute:Exec
    Target:Gameplay Event Router Component Object Reference=@n16
    EventType:Script Struct Object Reference=/Script/SparksCoreRuntime.SparksCoreAudioEvent_TempoChanged
    Context:Object Reference
    EventBubblingRule:EEventBubblingRule Enum
   ) -> (
    then:Exec
    AsyncTaskProxy:Async Action Start Listening To Stateful Event Object Reference
    OnEventReceived:Exec
    OnSavedEventStateExists:Exec
    OnEventStateCleared:Exec
    Payload:Sparks Core Audio Event Tempo Changed Structure
   ) {is=title K2Node_StartListeningToStatefulEventAsyncAction}
   > Set Tempo Changed Event Handle(Tempo Changed Event Handle:Async Action Start Listening To Stateful Event Object Reference=@n18) -> (Output_Get:Async Action Start Listening To Stateful Event Object Reference)
   n19 = OnTempoChanged!(
    execute=[@n18.OnEventReceived, @n18.OnSavedEventStateExists]
    self
    AudioEventTempoChanged:Sparks Core Audio Event Tempo Changed Structure=@n18.Payload
   ) -> (then) {is=title K2Node_CallFunction}
   UpdateResetTKMAvailable(execute=[@n21.then, @n23.then, @n19.then])
   n20 = StartListeningToStatefulEvent(execute=@n17.then_1, Target=@n16, EventType=/Script/SparksCoreRuntime.SparksCoreAudioEvent_KeyChanged) -> (Payload:Sparks Core Audio Event Key Changed Structure)
   > Set Key Changed Event Handle(Key Changed Event Handle:Async Action Start Listening To Stateful Event Object Reference=@n20) -> (Output_Get:Async Action Start Listening To Stateful Event Object Reference)
   n21 = OnKeyChanged!(
    execute=[@n20.OnEventReceived, @n20.OnSavedEventStateExists]
    self
    AudioEventKeyChanged:Sparks Core Audio Event Key Changed Structure=@n20.Payload
   ) -> (then) {is=title K2Node_CallFunction}
   n22 = StartListeningToStatefulEvent(execute=@n17.then_2, Target=@n16, EventType=/Script/SparksCoreRuntime.SparksCoreAudioEvent_KeyModeChanged) -> (Payload:Sparks Core Audio Event Key Mode Changed Structure)
   > Set Mode Changed Event Handle(Mode Changed Event Handle:Async Action Start Listening To Stateful Event Object Reference=@n22) -> (Output_Get:Async Action Start Listening To Stateful Event Object Reference)
   n23 = OnModeChanged!(
    execute=[@n22.OnEventReceived, @n22.OnSavedEventStateExists]
    self
    AudioEventKeyModeChanged:Sparks Core Audio Event Key Mode Changed Structure=@n22.Payload
   ) -> (then) {is=title K2Node_CallFunction}
   Event On Activated()
   > ShowHideLoopTool()
   > ShowHideKeySelector()
   n24 = GetJamSession!(self) -> (ReturnValue)
   > Branch(Condition=IsValidInterface(Interface=@n24))
   n25 = GetGameplayEventRouterComponent!(self=@n24) -> (ReturnValue)
   > Is Valid(InputObject=@n25)
   > n26 = Set EventRouter(EventRouter:Gameplay Event Router Component Object Reference=@n25) -> (Output_Get:Gameplay Event Router Component Object Reference)
   > Broadcast Event!(
    execute
    self:Gameplay Event Router Component Object Reference=@n26
    EventData:Jam Event Global Controls Displayed Structure=Make Jam Event Global Controls Displayed!(isDisplayed_1_#1:bool=true) -> (JamEvent_GlobalControls_Displayed:Jam Event Global Controls Displayed Structure) {is=title K2Node_MakeStruct}
    EventContext:Object Reference=GetOwningPlayer()
   ) -> (then, ReturnValue:bool)
   > ListenForTempoKeyModeChanges() {is=title K2Node_CallFunction}
   > n27 = Sequence!(execute) -> (then_0, then_1)
   > n28 = Cast To FortPlayerControllerZone!(execute, Object=GetOwningPlayer()) -> (then, CastFailed, AsFort Player Controller Zone:Fort Player Controller Zone Object Reference) {is=title K2Node_DynamicCast}
   > SetHUDElementVisibilityWithContext!(
    execute
    self:Fort Player Controller Object Reference=@n28
    HUDElementTags:"Gameplay Tag Container Structure (by ref)"=Get ReticleHUDElementTags() -> (ReticleHUDElementTags:Gameplay Tag Container Structure)
    bHideElements:bool=true
    ReasonName:Name=Reticle
    LogContext:String
   ) -> (then)
   n29 = WaitGameplayTagAddToActor!(
    execute:Exec=@n27.then_1
    TargetActor:Actor Object Reference=Get Controlled Pawn!(self:Controller Object Reference=GetOwningPlayer()) -> (ReturnValue:Pawn Object Reference) {is=title K2Node_CallFunction}
    Tag:Gameplay Tag Structure=(TagName="Pawn.HidePlayerPawn.Default")
    OnlyTriggerOnce:bool=false
   ) -> (then:Exec, AsyncTaskProxy:Ability Async Wait Gameplay Tag Added Object Reference, Added:Exec) {is=title K2Node_AsyncAction}
   Close(execute=@n29.Added)
   Event Construct()
   > SetFocus(self=Get TempoSelector_Module())
   > InitializeVariables()
   > ShowHideLoopTool()
   > SwitchLoopTool(Index=0)
   > ShowHideKeySelector()
   > ShowHidePrivacySelector()
   Event On Deactivated()
   > CleanUpMatchVocalsButton() {is=title K2Node_CallFunction}
   > n30 = Get!(execute) -> (then, else, EventRouter:Gameplay Event Router Component Object Reference)
   > Broadcast Event(
    self=@n30
    EventData=Make Jam Event Global Controls Displayed(isDisplayed_1_#1=False)
    EventContext=GetOwningPlayer()
   )
   > StopListeningForTempoKeyModeChanges() {is=title K2Node_CallFunction}
   > n31 = Cast To FortPlayerControllerZone(Object=GetOwningPlayer())
   > SetHUDElementVisibilityWithContext(self=@n31, HUDElementTags=Get ReticleHUDElementTags(), bHideElements=false, ReasonName=Reticle)
   Event PreConstruct()
   > SetDividerOrientation!(
    execute
    self:WBP UIKit Divider Base Object Reference=[
     Get UIKit_Divider_1() -> (UIKit_Divider_1:WBP UIKit Divider Base Object Reference)
     Get UIKit_Divider_Privacy() -> (UIKit_Divider_Privacy:WBP UIKit Divider Base Object Reference)
     Get UIKit_Divider() -> (UIKit_Divider:WBP UIKit Divider Base Object Reference)
    ]
    IsVertical:bool=true
   ) -> (then)
   > SetNavigationRuleBase!(
    execute
    self:Widget Object Reference=[
     Get OverlayPanel() -> (OverlayPanel:Overlay Object Reference)
     Get VerticalBox_ButtonStack() -> (VerticalBox_ButtonStack:Vertical Box Object Reference)
    ]
    Direction:EUINavigation Enum=Down
    Rule:EUINavigationRule Enum=Select(Option 0:EUINavigationRule Enum=Escape, Option 1:EUINavigationRule Enum=Stop, Index=Get ShowBackButton() -> (ShowBackButton:bool)) -> (ReturnValue:EUINavigationRule Enum)
   ) -> (then)
   > SwitchLoopTool(Index=0)
   Event OnInitialized()
   > SetMyText!(
    execute
    self:WBP FMJam Loop Button Object Reference=Get SetFromLoop_Button() -> (SetFromLoop_Button:WBP FMJam Loop Button Object Reference)
    InText:Text="LOCTEXT(\"Snap To Vocals\")"
   ) -> (then)
   > SetMyText(self=Get ConfirmLoop_Button() -> (ConfirmLoop_Button:WBP FMJam Loop Button Object Reference), InText="LOCTEXT(\"Confirm\")")
   n32 = SlotAsOverlaySlot(Widget=Get OverlayPanel())
   > Is Valid(InputObject=@n32)
   n33 = GetOwningPlayer()
   > Branch(Condition=AND Boolean(A=IsMobileGame(), B=IsUsingTouch(PlayerController=@n33)))
   > SetHorizontalAlignment(self=@n32, InHorizontalAlignment=HAlign_Right)
   > SetVerticalAlignment(self=@n32, InVerticalAlignment=VAlign_Bottom)
   > SetPadding!(
    execute
    self:Overlay Slot Object Reference=@n32
    InPadding_Left:float=0
    InPadding_Top:float=0
    InPadding_Right:float=150
    InPadding_Bottom:float=0
   ) -> (then)
   Set Visibility(execute=@n35.then, Visibility=Hidden, self=@n11)
   Set Visibility(execute=@n34.then, Visibility=Hidden, self=@n6)
   n34 = ChangeMode!(execute=@n4.then, self, NewMode:EMusicKeyMode Enum=Get New Mode() -> (New Mode:EMusicKeyMode Enum=Major)) -> (then) {is=title K2Node_CallFunction}
   n35 = ChangeKey!(execute=@n9.then, self, NewKey:EMusicKey Enum=Get New Key() -> (New Key:EMusicKey Enum=C)) -> (then) {is=title K2Node_CallFunction}
   n36 = Get SetFromLoop_Button()
   SetIsEnabled(execute=@n37.then, self=@n36, bInIsEnabled=false)
   > Delay(Duration=8)
   > SetIsEnabled(self=@n36, bInIsEnabled=true)
   > n37 = Delay(Duration=8)
  ]}
 InitMatchVocalsButton {nodes [
   n1 = InitMatchVocalsButton!() -> (then, MusicSlots:Array of Jam Music Slot VM Object References) {is=title K2Node_FunctionEntry}
   > n2 = For Each Loop(Array=@n1)
   UpdateResetTKMAvailable(execute=@n2.Completed)
   Add Field Value Changed Delegate!(
    execute=@n2.LoopBody
    self:MVVM Base Viewmodel Object Reference=@n2
    FieldId:Field Notification Id Structure=(FieldName="LoopType")
    Delegate:Delegate=Create Event()
   ) -> (then)
   > Add Field Value Changed Delegate(self=@n2, FieldId=(FieldName="SongShortName"), Delegate=Create Event())
  ]}
 InitializeVariables {nodes [
   InitializeVariables() {is=title K2Node_FunctionEntry}
   n1 = GetJamSession()
   > n2 = Branch(Condition=IsValidInterface(Interface=@n1))
   n3 = Get Component On JamSession!(
    ComponentClass:Actor Component Class Reference=/Script/FMJamSessionRuntime.JamSessionComponent_AudioState
    JamSession:Jam Session Interface Interface=@n1
   ) -> (OutFoundComponent:bool, ReturnValue:Jam Session Component Audio State Object Reference) {is=title K2Node_CallFunction}
   > n4 = Branch(Condition=@n3)
   > Set JamAudioState(JamAudioState:Jam Session Component Audio State Object Reference=@n3.ReturnValue) -> (Output_Get:Jam Session Component Audio State Object Reference)
   > Set JamMusicManagerComponent(JamMusicManagerComponent:Jam Session Component Music Manager Object Reference=GetMusicManager!(self:Jam Session Interface Interface=@n1) -> (ReturnValue:Jam Session Component Music Manager Object Reference)) -> (Output_Get:Jam Session Component Music Manager Object Reference)
   n5 = GetComponentByClass!(
    self:Actor Object Reference=GetOwningPlayer()
    ComponentClass:Actor Component Class Reference=/Script/FMJamSystemRuntime.JamControllerComponent
   ) -> (ReturnValue:Jam Controller Component Object Reference)
   > n6 = Is Valid(InputObject=@n5)
   > Set JamControllerComponent(JamControllerComponent:Jam Controller Component Object Reference=@n5) -> (Output_Get:Jam Controller Component Object Reference)
   > Return Node!(execute, Initialized:bool=true)
   Return Node(execute=[@n6.Is Not Valid, @n4.else, @n2.else], Initialized=false)
  ]}
 MatchVocals {nodes [
   MatchVocals() {is=title K2Node_FunctionEntry}
   n1 = Get Reset Song Short Name() -> (Reset Song Short Name:Name=None)
   > n2 = Branch(Condition="Not Equal (Name)"(A=@n1) -> (ReturnValue=~))
   > SetTempoAndKeyFromShortName(InShortName=@n1)
   > Return Node!(execute)
   FortShippingLog!(
    execute=@n2.else
    WorldContextObject:Object Reference
    InString:String=Reset Loops button pressed without a vaild song sort name
    bLogAsWarning:bool=true
   ) -> (then)
  ]}
 MusicSlotLoopTypeOrShortNameChanged {nodes [
   MusicSlotLoopTypeOrShortNameChanged!() -> (then, Object:Object Reference, Field:Field Notification Id Structure) {is=title K2Node_FunctionEntry}
   > UpdateResetTKMAvailable()
  ]}
 SetLoopPreview {nodes [
   n1 = SetLoopPreview!() -> (then, SongSelected:Object Reference) {is=title K2Node_FunctionEntry}
   > n2 = Cast To JamMusicSlotVM(Object=@n1)
   n3 = GetFMJamSongCatalog!() -> (ReturnValue:FMJam Song Catalog Object Reference)
   > n4 = Is Valid(InputObject=@n3)
   PrintString(execute=@n4.Is Not Valid, InString=SetTempoAndKeyFromShortName - ERROR getting Song Catalog, TextColor=(R=1,G=0,B=0,A=1))
   n5 = GetSongCatalogEntryByShortName!(self:FMJam Song Catalog Object Reference=@n3, ShortName:Name=Get SongShortName(self=@n2) -> (SongShortName=None)) -> (ReturnValue:Catalog Data Object Reference)
   Is Valid(exec=@n4.Is Valid, InputObject=@n5)
   n6 = Get!(self:Catalog Data Object Reference=@n5) -> (
    DataEntry_SongUID:String
    DataEntry_Shortname:String
    DataEntry_Title:String
    DataEntry_Artist:String
    DataEntry_Album:String
    DataEntry_Genre:Array of ESparksGenre Enums
    DataEntry_AnimGenre:Gameplay Tag Structure
    DataEntry_StageMood:Gameplay Tag Structure
    DataEntry_Duration:int
    DataEntry_CountInDuration:float=~
    DataEntry_MusicMomentOffsetSeconds:float=~
    DataEntry_ReleaseYear:int
    DataEntry_NewUntilTime:Date Time Structure
    DataEntry_AgeRating:ESparksSongAgeRating Enum
    DataEntry_MusicKey:EMusicKey Enum
    DataEntry_MusicMode:EMusicKeyMode Enum
    DataEntry_MusicTempo:int
    DataEntry_StartingInstrumentGuitar:ESparksInstrumentType Enum
    DataEntry_StartingInstrumentBass:ESparksInstrumentType Enum
    DataEntry_StartingInstrumentDrums:ESparksInstrumentType Enum
    DataEntry_StartingInstrumentVocals:ESparksInstrumentType Enum
    DataEntry_SanitizedSearchString:String
    DataEntry_JamLinkCode:String
    DataEntry_McpTemplateId:String
    DataEntry_AlbumArtUrl:String
    DataEntry_AlbumArtThumbnailUrl:String
    DataEntry_MidiUrl:String
    DataEntry_LipSyncUrl:String
    DataEntry_GameplayTagContainer:Gameplay Tag Container Structure
    DataEntry_QuicksilverInfo:Sparks Song Quicksilver Info Structure
    DataEntry_Intensities:Sparks Song Intensity Info Structure
    DataEntry_bValid:bool=~
   )
   > OverrideTempoEntry!(
    execute
    self:WBP FMJam Tempo Selector Object Reference=Get TempoSelector_LoopPreview() -> (TempoSelector_LoopPreview:WBP FMJam Tempo Selector Object Reference)
    TempoPreview:int=@n6.DataEntry_MusicTempo
   ) -> (then)
   > OverrideKeyEntry!(
    execute
    self:WBP FMJam Key Preview Object Reference=Get JamKeyPreview_LoopPreview() -> (JamKeyPreview_LoopPreview:WBP FMJam Key Preview Object Reference)
    MusicKeyPreview:EMusicKey Enum=@n6.DataEntry_MusicKey
   ) -> (then)
   > OverrideModeEntry!(
    execute
    self:WBP FMJam Mode Selector Object Reference=Get ModeSelector_LoopPreview() -> (ModeSelector_LoopPreview:WBP FMJam Mode Selector Object Reference)
    ModePreview:EMusicKeyMode Enum=@n6.DataEntry_MusicMode
   ) -> (then)
  ]}
 SetTempoAndKeyFromShortName {nodes [
   n1 = SetTempoAndKeyFromShortName!() -> (then, InShortName:Name) {is=title K2Node_FunctionEntry}
   n2 = GetFMJamSongCatalog()
   > n3 = Is Valid(InputObject=@n2)
   PrintString(execute=@n3.Is Not Valid, InString=SetTempoAndKeyFromShortName - ERROR getting Song Catalog, TextColor=(R=1,G=0,B=0,A=1))
   n4 = GetSongCatalogEntryByShortName(self=@n2, ShortName=@n1)
   n5 = Is Valid(exec=@n3.Is Valid, InputObject=@n4)
   n6 = Get InShortName() -> (InShortName:Name=None)
   PrintString(
    execute=@n5.Is Not Valid
    InString=Append(A="SetTempoAndKeyFromShortName - ERROR getting Song Catalog Entry for: ", B="To String (Name)"(InName=@n6))
    TextColor=(R=1,G=0,B=0,A=1)
   )
   n7 = Get(self=@n4) -> (DataEntry_CountInDuration=~, DataEntry_MusicMomentOffsetSeconds=~, DataEntry_bValid=~)
   ChangeTempo(execute=@n5.Is Valid, NewTempo=@n7.DataEntry_MusicTempo)
   > ChangeKey(NewKey=@n7.DataEntry_MusicKey)
   > ChangeMode(NewMode=@n7.DataEntry_MusicMode)
   n8 = Get InShortName() -> (InShortName=None)
   > PrintString(
    InString=Append!(
     A="SetTempoAndKeyFromShortName for "
     B="To String (Name)"(InName=@n8)
     C:String="-- Tempo: "
     D:String="To String (Integer)"(InInt=@n7.DataEntry_MusicTempo)
     E:String="-- Key: "
     F:String=Enum to String!(Enumerator:EMusicKey Enum=@n7.DataEntry_MusicKey) -> (ReturnValue:String)
    ) -> (ReturnValue)
    TextColor=(R=0,G=1,B=0.5739,A=1)
   )
  ]}
 ShowHideKeySelector {nodes [
   ShowHideKeySelector() {is=title K2Node_FunctionEntry}
   > n1 = ShouldShowGlobalControlsKeySelector!(execute) -> (then, ReturnValue:bool)
   > n2 = Branch(Condition=@n1)
   > SetVisibility(self=Get KeyModule() -> (KeyModule:Overlay Object Reference), InVisibility=SelfHitTestInvisible)
   > "SetText (Text)"(self=Get Text_KeyModule() -> (Text_KeyModule:Common Text Object Reference), InText="LOCTEXT(\"Mode\")")
   n3 = Get KeyPreviewPanel() -> (KeyPreviewPanel:Fort Responsive Panel Object Reference)
   > SetVisibility(self=@n3, InVisibility=Collapsed)
   > SetPadding(
    self:Horizontal Box Slot Object Reference=SlotAsHorizontalBoxSlot!(Widget:Widget Object Reference=@n3) -> (ReturnValue:Horizontal Box Slot Object Reference)
    InPadding_Left=10
    InPadding_Top=0
    InPadding_Right=0
    InPadding_Bottom=0
   )
   SetVisibility(execute=@n2.else, self=Get KeyModule(), InVisibility=Collapsed)
   > "SetText (Text)"(self=Get Text_KeyModule(), InText="LOCTEXT(\"Key\")")
   n4 = Get KeyPreviewPanel()
   > SetVisibility(self=@n4, InVisibility=SelfHitTestInvisible)
   > SetPadding(self=SlotAsHorizontalBoxSlot(Widget=@n4), InPadding_Left=0, InPadding_Top=0, InPadding_Right=0, InPadding_Bottom=0)
  ]}
 ShowHideLoopTool {nodes [
   ShowHideLoopTool() {is=title K2Node_FunctionEntry}
   > n1 = ShouldShowGlobalControlsSetFromLoopButton!(execute) -> (then, ReturnValue:bool)
   > n2 = Branch(Condition=@n1)
   > SetVisibility(self=Get SetFromLoop_Button(), InVisibility=SelfHitTestInvisible)
   > SetVisibility(self=Get UIKit_Divider(), InVisibility=SelfHitTestInvisible)
   SetVisibility(execute=@n2.else, self=Get SetFromLoop_Button(), InVisibility=Collapsed)
   > SetVisibility(self=Get UIKit_Divider(), InVisibility=Collapsed)
  ]}
 ShowHidePrivacySelector {nodes [
   ShowHidePrivacySelector() {is=title K2Node_FunctionEntry}
   > n1 = IsJamPrivacyEnabled!(execute) -> (then, ReturnValue:bool)
   n2 = Get isPrivacySelectorAllowed() -> (isPrivacySelectorAllowed:bool)
   > SetVisibility(
    self=[Get UIKit_Divider_Privacy(), Get WBP_FMJamPrivacySelector() -> (WBP_FMJamPrivacySelector:WBP FMJam Privacy Selector Object Reference)]
    InVisibility=Select(Option 0:ESlateVisibility Enum=Collapsed, Option 1:ESlateVisibility Enum=SelfHitTestInvisible, Index=AND Boolean(A=@n1, B=@n2)) -> (ReturnValue:ESlateVisibility Enum)
   )
  ]}
 SwitchLoopTool {nodes [
   n1 = SwitchLoopTool!() -> (then, Index:int) {is=title K2Node_FunctionEntry}
   > n2 = Switch on Int!(execute:Exec, Selection:int=@n1) -> (Default:Exec, 0:Exec, 1:Exec, 2:Exec)
   n3 = Branch(execute=@n2.0, Condition=Get ShowBackButton())
   > n4 = SetActiveWidgetIndex(
    self:Common Visibility Switcher Object Reference=[
     Get LoopToolSwitcher() -> (LoopToolSwitcher:Common Visibility Switcher Object Reference)
     Get BackButtonSwitcher() -> (BackButtonSwitcher:Common Visibility Switcher Object Reference)
    ]
    Index=0
   )
   SetFocus(execute=[@n5.then, @n4.then], self=Get TempoSelector_Module())
   > Set SongSelected()
   > Refresh(self:WBP FMJam Tempo Selector Object Reference=Get TempoSelector_Module())
   > Refresh(self:WBP FMJam Key Preview Object Reference=Get WBP_FMJamKeyPreview() -> (WBP_FMJamKeyPreview:WBP FMJam Key Preview Object Reference))
   > Refresh(self:WBP FMJam Mode Selector Object Reference=Get ModeSelector_Module())
   SetActiveWidgetIndex(execute=@n3.else, self=Get LoopToolSwitcher(), Index=0)
   > n5 = SetActiveWidgetIndex(self=Get BackButtonSwitcher(), Index=3)
   SetActiveWidgetIndex(execute=@n2.1, self=[Get BackButtonSwitcher(), Get LoopToolSwitcher()], Index=1)
   > SetFocus(self=Get CommonListView_53() -> (CommonListView_53:Common List View Object Reference))
   > Set SongSelected()
   SetActiveWidgetIndex(execute=@n2.2, self=[Get LoopToolSwitcher(), Get BackButtonSwitcher()], Index=2)
   > SetFocus(self=Get ConfirmLoop_Button())
  ]}
 UpdateResetTKMAvailable {nodes [
   UpdateResetTKMAvailable() {is=title K2Node_FunctionEntry}
   > n1 = Get!(execute:Exec) -> (then:Exec, else:Exec, JamMusicManagerComponent:Jam Session Component Music Manager Object Reference)
   > n2 = For Each Loop with Break!(
    Exec:Exec
    Array:Array of Jam Music Slot Object References=GetAllMusicSlots!(self:Jam Session Component Music Manager Object Reference=@n1) -> (ReturnValue:Array of Jam Music Slot Object References)
    Break:Exec
   ) -> (LoopBody:Exec, Array Element:Jam Music Slot Object Reference, Array Index:int, Completed:Exec)
   > Branch(Condition="Equal (Enum)"(
     A:EFMJamLoopType Enum=GetCurrentLoopType!(self:Jam Music Slot Object Reference=@n2) -> (ReturnValue:EFMJamLoopType Enum=Lead)
     B:EFMJamLoopType Enum=Lead
    ))
   > n3 = Set Reset Song Short Name(Reset Song Short Name:Name=GetCurrentSongShortName!(self:Jam Music Slot Object Reference=@n2) -> (ReturnValue:Name=None)) -> (Output_Get:Name=None)
   > n4 = GetJamSessionForActor!(execute, Actor:Actor Object Reference=GetOwningPlayer()) -> (then, ReturnValue:Jam Session Interface Interface)
   n5 = TryGetKeyModeTempo!(self:Jam Session Component Music Manager Object Reference=GetMusicManager(self=@n4)) -> (OutMusicKey:EMusicKey Enum=C, OutMusicKeyMode:EMusicKeyMode Enum=Major, OutTempo:int=0, ReturnValue:bool)
   n6 = Get!(self:Catalog Data Object Reference=GetSongCatalogEntryByShortName(self=GetFMJamSongCatalog(), ShortName=@n3)) -> (
    DataEntry_SongUID:String
    DataEntry_Shortname:String
    DataEntry_Title:String
    DataEntry_Artist:String
    DataEntry_Album:String
    DataEntry_Genre:Array of ESparksGenre Enums
    DataEntry_AnimGenre:Gameplay Tag Structure
    DataEntry_StageMood:Gameplay Tag Structure
    DataEntry_Duration:int
    DataEntry_CountInDuration:float=~
    DataEntry_MusicMomentOffsetSeconds:float=~
    DataEntry_ReleaseYear:int
    DataEntry_NewUntilTime:Date Time Structure
    DataEntry_AgeRating:ESparksSongAgeRating Enum
    DataEntry_MusicKey:EMusicKey Enum
    DataEntry_MusicMode:EMusicKeyMode Enum
    DataEntry_MusicTempo:int
    DataEntry_StartingInstrumentGuitar:ESparksInstrumentType Enum
    DataEntry_StartingInstrumentBass:ESparksInstrumentType Enum
    DataEntry_StartingInstrumentDrums:ESparksInstrumentType Enum
    DataEntry_StartingInstrumentVocals:ESparksInstrumentType Enum
    DataEntry_SanitizedSearchString:String
    DataEntry_JamLinkCode:String
    DataEntry_McpTemplateId:String
    DataEntry_AlbumArtUrl:String
    DataEntry_AlbumArtThumbnailUrl:String
    DataEntry_MidiUrl:String
    DataEntry_LipSyncUrl:String
    DataEntry_GameplayTagContainer:Gameplay Tag Container Structure
    DataEntry_QuicksilverInfo:Sparks Song Quicksilver Info Structure
    DataEntry_Intensities:Sparks Song Intensity Info Structure
    DataEntry_bValid:bool=~
   )
   n7 = "Equal (Enum)"(A:EMusicKey Enum=@n5, B:EMusicKey Enum=@n6.DataEntry_MusicKey)
   n8 = "Equal (Enum)"(A:EMusicKeyMode Enum=@n5.OutMusicKeyMode, B:EMusicKeyMode Enum=@n6.DataEntry_MusicMode)
   n9 = "Equal (Integer)"(A=@n5.OutTempo, B=@n6.DataEntry_MusicTempo)
   > SetCanReset!(
    execute
    self:WBP FMJam Loop Button Object Reference=Get SetFromLoop_Button()
    CanReset:bool=NOT Boolean(A=AND Boolean!(A=@n7, B=@n8, C:bool=@n9, D:bool=@n5.ReturnValue) -> (ReturnValue))
   ) -> (then)
   > Return Node()
   Set Reset Song Short Name(execute=[@n2.Completed, @n1.else]) -> (Output_Get=None)
   > SetCanReset(self=Get SetFromLoop_Button(), CanReset=false)
   > Return Node()
  ]}
]
variables [
 Button {type=CommonButtonBase, default=null}
 EventRouter {type=GameplayEventRouterComponent, default=null}
 isPrivacySelectorAllowed {type=boolean, default=true}
 JamAudioState {type=JamSessionComponent_AudioState, default=null}
 JamControllerComponent {type=JamControllerComponent, default=null}
 JamMusicManagerComponent {type=JamSessionComponent_MusicManager, default=null}
 Key Changed Event Handle {type=AsyncAction_StartListeningToStatefulEvent, default=null}
 KeyChangeTimer {type=TimerHandle}
 Mode Changed Event Handle {type=AsyncAction_StartListeningToStatefulEvent, default=null}
 ModeChangeTimer {type=TimerHandle}
 New Key {type=EMusicKey, default=C}
 New Mode {type=EMusicKeyMode, default=Major}
 Reset Song Short Name {type=string, default=None}
 ReticleHUDElementTags {type=GameplayTagContainer, default {gameplayTags [{tagName=HUD.Center.Reticle}, {tagName=HUD.InWorld.ActorCanvas}]}}
 SettleBeforeCountTime {
  type=ScalableFloat
  default {registryType=None, value=1, curve {curveTable=/FMJamSystem/DataTables/PilgrimJamGameData, rowName=Jam.GlobalControls.SettleTime}}
 }
 ShowBackButton {type=boolean, default=true}
 SongSelected {type=Object, default=null}
 Tempo Changed Event Handle {type=AsyncAction_StartListeningToStatefulEvent, default=null}
]
widgets {
 info {inheritedWidgetCount=1, parentClass=/Script/FMJamSystemRuntime.JamGlobalControlsWidget, rootWidgetClass=&Overlay, widgetCount=65}
 widgets [
  Overlay_20 {widgetClassPath=&Overlay, widget=Overlay_20}
  OverlayPanel {
   widgetClassPath=&Overlay
   bIsVariable=true
   parent=Overlay_20
   widget=OverlayPanel
   properties {navigation=&self:WidgetTree.OverlayPanel.WidgetNavigation_0}
   slot {class=OverlaySlot, properties {verticalAlignment=VAlign_Center}}
  }
  SafeZone_0 {
   widgetClassPath=&SafeZone
   bIsVariable=true
   parent=OverlayPanel
   widget=SafeZone_0
   properties {padBottom=false, padLeft=false, padRight=false, padTop=false}
   slot {class=OverlaySlot, properties {verticalAlignment=VAlign_Bottom}}
  }
  SizeBox_1 {
   widgetClassPath=&SizeBox
   parent=SafeZone_0
   widget=SizeBox_1
   properties {
    bOverride_MaxDesiredWidth=true
    bOverride_MinDesiredWidth=true
    heightOverride=720
    maxDesiredWidth=400
    minDesiredWidth=300
    visibility=Visible
   }
   slot {class=SafeZoneSlot}
  }
  VerticalBox_223 {
   widgetClassPath=&VerticalBox
   parent=SizeBox_1
   widget=VerticalBox_223
   slot {class=SizeBoxSlot, properties {verticalAlignment=VAlign_Bottom}}
  }
  Overlay_40 {
   widgetClassPath=&Overlay
   parent=VerticalBox_223
   widget=Overlay_40
   slot {class=VerticalBoxSlot, properties {horizontalAlignment=HAlign_Left, padding=<ltrb 10 10 10 0>}}
  }
  Border_38 {
   widgetClassPath=&Border
   parent=Overlay_40
   widget=Border_38
   properties {
    background {
     bIsDynamicallyLoaded=false
     drawAs=RoundedBox
     imageType=NoImage
     mirroring=NoMirror
     resourceName=None
     resourceObject=null
     tiling=NoTile
     imageSize=<32 32>
     margin=<ltrb 0 0 0 0>
     outlineSettings {
      bUseBrushTransparency=false
      roundingType=FixedRadius
      width=0
      color {colorUseRule=UseColor_Specified, specifiedColor=<rgba 0 0 0 0>}
      cornerRadii=<17 17 17 17>
     }
     tintColor {colorUseRule=UseColor_Specified, specifiedColor=<rgba 0 0 0 0.8>}
     uVRegion {bIsValid=false, max=<0 0>, min=<0 0>}
    }
    padding=<ltrb 0 0 0 0>
   }
   slot {class=OverlaySlot, properties {horizontalAlignment=HAlign_Fill, verticalAlignment=VAlign_Fill}}
  }
  PostBufferUpdate_301 {
   widgetClassPath=/Script/UMG.PostBufferUpdate
   bIsVariable=true
   parent=Overlay_40
   widget=PostBufferUpdate_301
   properties {buffersToUpdate [ESlatePostRT_0]}
   slot {class=OverlaySlot, properties {horizontalAlignment=HAlign_Fill, verticalAlignment=VAlign_Fill}}
  }
  VerticalBox_ButtonStack {
   widgetClassPath=&VerticalBox
   bIsVariable=true
   parent=Overlay_40
   widget=VerticalBox_ButtonStack
   properties {navigation=&self:WidgetTree.VerticalBox_ButtonStack.WidgetNavigation_0}
   slot {class=OverlaySlot, properties {verticalAlignment=VAlign_Bottom, padding=<ltrb 20 20 20 20>}}
  }
  LoopToolSwitcher {
   widgetClassPath=&CommonVisibilitySwitcher
   bIsVariable=true
   parent=VerticalBox_ButtonStack
   widget=LoopToolSwitcher
   slot {class=VerticalBoxSlot, properties {horizontalAlignment=HAlign_Center, verticalAlignment=VAlign_Bottom}}
  }
  TempoKey_VerticalBox {
   widgetClassPath=&VerticalBox
   parent=LoopToolSwitcher
   widget=TempoKey_VerticalBox
   properties {bOverride_Cursor=true}
   slot {class=CommonVisibilitySwitcherSlot, properties {verticalAlignment=VAlign_Bottom}}
  }
  HorizontalBox_1 {widgetClassPath=&HorizontalBox, parent=TempoKey_VerticalBox, widget=HorizontalBox_1, slot {class=VerticalBoxSlot}}
  TempoSelector_Module {
   widgetClassPath=&/WBP_FMJamTempoSelector
   bIsVariable=true
   parent=HorizontalBox_1
   widget=TempoSelector_Module
   slot {class=HorizontalBoxSlot, properties {size {sizeRule=Fill, value=1}}}
  }
  Spacer_93 {widgetClassPath=&Spacer, parent=TempoKey_VerticalBox, widget=Spacer_93, properties {size=<1 10>}, slot {class=VerticalBoxSlot}}
  KeyModule {
   widgetClassPath=&Overlay
   bIsVariable=true
   parent=TempoKey_VerticalBox
   widget=KeyModule
   properties {visibility=Collapsed}
   slot {class=VerticalBoxSlot, properties {verticalAlignment=VAlign_Bottom}}
  }
  DEACTIVATEDKeySelector_Module {
   widgetClassPath=&/WBP_FMJamKeySelector
   bIsVariable=true
   parent=KeyModule
   widget=DEACTIVATEDKeySelector_Module
   slot {class=OverlaySlot, properties {horizontalAlignment=HAlign_Fill, verticalAlignment=VAlign_Fill}}
  }
  KeySelectorTimer_Image {
   widgetClassPath=&Image
   bIsVariable=true
   parent=KeyModule
   widget=KeySelectorTimer_Image
   properties {
    visibility=Hidden
    brush {
     bIsDynamicallyLoaded=false
     drawAs=Image
     imageType=NoImage
     mirroring=NoMirror
     resourceName=None
     resourceObject=/Game/Athena/HUD/Materials/M_UI_Timer
     tiling=NoTile
     imageSize=<32 32>
     margin=<ltrb 0 0 0 0>
     outlineSettings {
      bUseBrushTransparency=false
      roundingType=HalfHeightRadius
      width=0
      color {colorUseRule=UseColor_Specified, specifiedColor=<rgba 0 0 0 0>}
      cornerRadii=<0 0 0 0>
     }
     tintColor {colorUseRule=UseColor_Specified, specifiedColor=<rgba 1 1 1 1>}
     uVRegion {bIsValid=false, max=<0 0>, min=<0 0>}
    }
   }
   slot {class=OverlaySlot, properties {horizontalAlignment=HAlign_Right, padding=<ltrb 0 0 10 15>}}
  }
  Spacer {widgetClassPath=&Spacer, parent=TempoKey_VerticalBox, widget=Spacer, properties {size=<1 10>}, slot {class=VerticalBoxSlot}}
  ModeModule {
   widgetClassPath=&Overlay
   parent=TempoKey_VerticalBox
   widget=ModeModule
   slot {class=VerticalBoxSlot, properties {verticalAlignment=VAlign_Bottom}}
  }
  ModeSelectorTimer_Image {
   widgetClassPath=&Image
   bIsVariable=true
   parent=ModeModule
   widget=ModeSelectorTimer_Image
   properties {
    visibility=Hidden
    brush {
     bIsDynamicallyLoaded=false
     drawAs=Image
     imageType=NoImage
     mirroring=NoMirror
     resourceName=None
     resourceObject=&M_UI_Timer
     tiling=NoTile
     imageSize=<32 32>
     margin=<ltrb 0 0 0 0>
     outlineSettings {
      bUseBrushTransparency=false
      roundingType=HalfHeightRadius
      width=0
      color {colorUseRule=UseColor_Specified, specifiedColor=<rgba 0 0 0 0>}
      cornerRadii=<0 0 0 0>
     }
     tintColor {colorUseRule=UseColor_Specified, specifiedColor=<rgba 1 1 1 1>}
     uVRegion {bIsValid=false, max=<0 0>, min=<0 0>}
    }
   }
   slot {class=OverlaySlot, properties {horizontalAlignment=HAlign_Right, padding=<ltrb 0 0 10 0>}}
  }
  VerticalBox_1 {
   widgetClassPath=&VerticalBox
   parent=ModeModule
   widget=VerticalBox_1
   properties {bOverride_Cursor=true}
   slot {class=OverlaySlot}
  }
  FortResponsivePanel_3 {
   widgetClassPath=/Script/FortniteUI.FortResponsivePanel
   parent=VerticalBox_1
   widget=FortResponsivePanel_3
   slot {class=VerticalBoxSlot, properties {padding=<ltrb 10 0 10 0>}}
  }
  SizeBox_0 {
   widgetClassPath=&SizeBox
   parent=FortResponsivePanel_3
   widget=SizeBox_0
   properties {minDesiredHeight=50}
   slot {class=FortResponsivePanelSlot}
  }
  Text_KeyModule {
   widgetClassPath=&CommonTextBlock
   bIsVariable=true
   parent=SizeBox_0
   widget=Text_KeyModule
   properties {
    justification=Center
    text=Key
    font {
     bForceMonospaced=false
     bMaterialIsStencil=false
     fontMaterial=null
     fontObject=&HeadingNow
     letterSpacing=0
     monospacedWidth=1
     size=16
     skewAmount=0
     typefaceFontName=75 Medium
     outlineSettings {
      bApplyOutlineToDropShadows=false
      bMiteredCorners=false
      bSeparateFillAlpha=false
      outlineMaterial=null
      outlineSize=0
      outlineColor=<rgba 0 0 0 1>
     }
    }
    shadowOffset=<0 0>
    strikeBrush {
     bIsDynamicallyLoaded=false
     drawAs=NoDrawType
     imageType=NoImage
     mirroring=NoMirror
     resourceName=None
     resourceObject=null
     tiling=NoTile
     imageSize=<32 32>
     margin=<ltrb 0 0 0 0>
     outlineSettings {
      bUseBrushTransparency=false
      roundingType=HalfHeightRadius
      width=0
      color {colorUseRule=UseColor_Specified, specifiedColor=<rgba 0 0 0 0>}
      cornerRadii=<0 0 0 0>
     }
     tintColor {colorUseRule=UseColor_Specified, specifiedColor=<rgba 1 1 1 1>}
     uVRegion {bIsValid=false, max=<0 0>, min=<0 0>}
    }
   }
   slot {class=SizeBoxSlot, properties {horizontalAlignment=HAlign_Center}}
  }
  SettingName_Spacer {
   widgetClassPath=&Spacer
   bIsVariable=true
   parent=VerticalBox_1
   widget=SettingName_Spacer
   properties {size=<1 10>}
   slot {class=VerticalBoxSlot}
  }
  KeyPreviewContainer {widgetClassPath=&HorizontalBox, parent=VerticalBox_1, widget=KeyPreviewContainer, slot {class=VerticalBoxSlot}}
  KeyPreviewPanel {
   widgetClassPath=&FortResponsivePanel
   bIsVariable=true
   parent=KeyPreviewContainer
   widget=KeyPreviewPanel
   slot {class=HorizontalBoxSlot, properties {verticalAlignment=VAlign_Center}}
  }
  WBP_FMJamKeyPreview {
   widgetClassPath=&/WBP_FMJamKeyPreview
   bIsVariable=true
   parent=KeyPreviewPanel
   widget=WBP_FMJamKeyPreview
   slot {class=FortResponsivePanelSlot}
  }
  ModeSelector_Module {
   widgetClassPath=&/WBP_FMJamModeSelector
   bIsVariable=true
   parent=KeyPreviewContainer
   widget=ModeSelector_Module
   properties {hide Text Block=true}
   slot {class=HorizontalBoxSlot, properties {verticalAlignment=VAlign_Bottom, size {sizeRule=Fill, value=1}}}
  }
  UIKit_Divider_Privacy {
   widgetClassPath=/FortUILibrary/Systems/Tabs/WBP_UIKit_Divider_Base
   bIsVariable=true
   parent=TempoKey_VerticalBox
   widget=UIKit_Divider_Privacy
   properties {bOverride_Cursor=true, div Width=2}
   slot {class=VerticalBoxSlot, properties {padding=<ltrb 0 15 0 15>}}
  }
  WBP_FMJamPrivacySelector {
   widgetClassPath=&/WBP_FMJamPrivacySelector
   bIsVariable=true
   parent=TempoKey_VerticalBox
   widget=WBP_FMJamPrivacySelector
   slot {class=VerticalBoxSlot}
  }
  UIKit_Divider {
   widgetClassPath=&WBP_UIKit_Divider_Base
   bIsVariable=true
   parent=TempoKey_VerticalBox
   widget=UIKit_Divider
   properties {bOverride_Cursor=true, div Width=2}
   slot {class=VerticalBoxSlot, properties {padding=<ltrb 0 15 0 0>}}
  }
  FortResponsivePanel_5 {
   widgetClassPath=&FortResponsivePanel
   parent=TempoKey_VerticalBox
   widget=FortResponsivePanel_5
   slot {class=VerticalBoxSlot, properties {verticalAlignment=VAlign_Bottom, padding=<ltrb 10 15 10 0>}}
  }
  SetFromLoop_Button {
   widgetClassPath=&/WBP_FMJamLoopButton
   bIsVariable=true
   parent=FortResponsivePanel_5
   widget=SetFromLoop_Button
   properties {bNavigateToNextWidgetOnDisable=true, bOverride_Cursor=true, colorAndOpacity=<rgba 1 1 1 1>}
   slot {class=FortResponsivePanelSlot}
  }
  SongSelector_VerticalBox {
   widgetClassPath=&VerticalBox
   bIsVariable=true
   parent=LoopToolSwitcher
   widget=SongSelector_VerticalBox
   slot {class=CommonVisibilitySwitcherSlot, properties {horizontalAlignment=HAlign_Center, verticalAlignment=VAlign_Center}}
  }
  Text_SettingName {
   widgetClassPath=&CommonTextBlock
   bIsVariable=true
   parent=SongSelector_VerticalBox
   widget=Text_SettingName
   properties {
    text=Set to
    font {
     bForceMonospaced=false
     bMaterialIsStencil=false
     fontMaterial=null
     fontObject=&HeadingNow
     letterSpacing=0
     monospacedWidth=1
     size=16
     skewAmount=0
     typefaceFontName=75 Medium
     outlineSettings {
      bApplyOutlineToDropShadows=false
      bMiteredCorners=false
      bSeparateFillAlpha=false
      outlineMaterial=null
      outlineSize=0
      outlineColor=<rgba 0 0 0 1>
     }
    }
    shadowOffset=<0 0>
    strikeBrush {
     bIsDynamicallyLoaded=false
     drawAs=NoDrawType
     imageType=NoImage
     mirroring=NoMirror
     resourceName=None
     resourceObject=null
     tiling=NoTile
     imageSize=<32 32>
     margin=<ltrb 0 0 0 0>
     outlineSettings {
      bUseBrushTransparency=false
      roundingType=HalfHeightRadius
      width=0
      color {colorUseRule=UseColor_Specified, specifiedColor=<rgba 0 0 0 0>}
      cornerRadii=<0 0 0 0>
     }
     tintColor {colorUseRule=UseColor_Specified, specifiedColor=<rgba 1 1 1 1>}
     uVRegion {bIsValid=false, max=<0 0>, min=<0 0>}
    }
   }
   slot {class=VerticalBoxSlot, properties {horizontalAlignment=HAlign_Center}}
  }
  Spacer_59 {
   widgetClassPath=&Spacer
   parent=SongSelector_VerticalBox
   widget=Spacer_59
   properties {size=<341 10>}
   slot {class=VerticalBoxSlot}
  }
  CommonListView_53 {
   widgetClassPath=&CommonListView
   bIsVariable=true
   parent=SongSelector_VerticalBox
   widget=CommonListView_53
   properties {
    bReturnFocusToSelection=true
    bSelectItemOnNavigation=false
    clipping=Inherit
    entryWidgetClass=&/WBP_FMJamLoopButton.WBP_FMJamLoopButton_C
    numDesignerPreviewEntries=4
    verticalEntrySpacing=15
   }
   slot {class=VerticalBoxSlot, properties {padding=<ltrb 10 0 10 0>}}
  }
  ConfirmSelection_VerticalBox {
   widgetClassPath=&VerticalBox
   parent=LoopToolSwitcher
   widget=ConfirmSelection_VerticalBox
   properties {bOverride_Cursor=true}
   slot {class=CommonVisibilitySwitcherSlot}
  }
  HorizontalBox_2 {
   widgetClassPath=&HorizontalBox
   parent=ConfirmSelection_VerticalBox
   widget=HorizontalBox_2
   properties {visibility=HitTestInvisible}
   slot {class=VerticalBoxSlot}
  }
  TempoSelector_LoopPreview {
   widgetClassPath=&WBP_FMJamTempoSelector
   bIsVariable=true
   parent=HorizontalBox_2
   widget=TempoSelector_LoopPreview
   properties {bIsEnabled=false}
   slot {class=HorizontalBoxSlot, properties {size {sizeRule=Fill, value=1}}}
  }
  Spacer_1 {
   widgetClassPath=&Spacer
   parent=ConfirmSelection_VerticalBox
   widget=Spacer_1
   properties {visibility=HitTestInvisible, size=<1 10>}
   slot {class=VerticalBoxSlot}
  }
  Overlay {
   widgetClassPath=&Overlay
   parent=ConfirmSelection_VerticalBox
   widget=Overlay
   properties {visibility=Collapsed}
   slot {class=VerticalBoxSlot}
  }
  DEACTIVATEDKeySelector_Module_1 {
   widgetClassPath=&WBP_FMJamKeySelector
   bIsVariable=true
   parent=Overlay
   widget=DEACTIVATEDKeySelector_Module_1
   slot {class=OverlaySlot, properties {horizontalAlignment=HAlign_Fill, verticalAlignment=VAlign_Fill}}
  }
  KeySelectorTimer_Image_1 {
   widgetClassPath=&Image
   bIsVariable=true
   parent=Overlay
   widget=KeySelectorTimer_Image_1
   properties {
    visibility=Hidden
    brush {
     bIsDynamicallyLoaded=false
     drawAs=Image
     imageType=NoImage
     mirroring=NoMirror
     resourceName=None
     resourceObject=&M_UI_Timer
     tiling=NoTile
     imageSize=<32 32>
     margin=<ltrb 0 0 0 0>
     outlineSettings {
      bUseBrushTransparency=false
      roundingType=HalfHeightRadius
      width=0
      color {colorUseRule=UseColor_Specified, specifiedColor=<rgba 0 0 0 0>}
      cornerRadii=<0 0 0 0>
     }
     tintColor {colorUseRule=UseColor_Specified, specifiedColor=<rgba 1 1 1 1>}
     uVRegion {bIsValid=false, max=<0 0>, min=<0 0>}
    }
   }
   slot {class=OverlaySlot, properties {horizontalAlignment=HAlign_Right, verticalAlignment=VAlign_Bottom, padding=<ltrb 0 0 10 15>}}
  }
  Spacer_2 {
   widgetClassPath=&Spacer
   parent=ConfirmSelection_VerticalBox
   widget=Spacer_2
   properties {visibility=HitTestInvisible, size=<1 10>}
   slot {class=VerticalBoxSlot}
  }
  Overlay_2 {widgetClassPath=&Overlay, parent=ConfirmSelection_VerticalBox, widget=Overlay_2, slot {class=VerticalBoxSlot}}
  ModeSelectorTimer_Image_1 {
   widgetClassPath=&Image
   bIsVariable=true
   parent=Overlay_2
   widget=ModeSelectorTimer_Image_1
   properties {
    visibility=Hidden
    brush {
     bIsDynamicallyLoaded=false
     drawAs=Image
     imageType=NoImage
     mirroring=NoMirror
     resourceName=None
     resourceObject=&M_UI_Timer
     tiling=NoTile
     imageSize=<32 32>
     margin=<ltrb 0 0 0 0>
     outlineSettings {
      bUseBrushTransparency=false
      roundingType=HalfHeightRadius
      width=0
      color {colorUseRule=UseColor_Specified, specifiedColor=<rgba 0 0 0 0>}
      cornerRadii=<0 0 0 0>
     }
     tintColor {colorUseRule=UseColor_Specified, specifiedColor=<rgba 1 1 1 1>}
     uVRegion {bIsValid=false, max=<0 0>, min=<0 0>}
    }
   }
   slot {class=OverlaySlot, properties {horizontalAlignment=HAlign_Right, padding=<ltrb 0 0 10 0>}}
  }
  VerticalBox_2 {
   widgetClassPath=&VerticalBox
   parent=Overlay_2
   widget=VerticalBox_2
   properties {bOverride_Cursor=true, visibility=HitTestInvisible}
   slot {class=OverlaySlot}
  }
  FortResponsivePanel_2 {
   widgetClassPath=&FortResponsivePanel
   parent=VerticalBox_2
   widget=FortResponsivePanel_2
   slot {class=VerticalBoxSlot, properties {padding=<ltrb 10 0 10 0>}}
  }
  SizeBox_2 {widgetClassPath=&SizeBox, parent=FortResponsivePanel_2, widget=SizeBox_2, slot {class=FortResponsivePanelSlot}}
  Text_KeyModule_1 {
   widgetClassPath=&CommonTextBlock
   bIsVariable=true
   parent=SizeBox_2
   widget=Text_KeyModule_1
   properties {
    justification=Center
    text=Key
    font {
     bForceMonospaced=false
     bMaterialIsStencil=false
     fontMaterial=null
     fontObject=&HeadingNow
     letterSpacing=0
     monospacedWidth=1
     size=16
     skewAmount=0
     typefaceFontName=75 Medium
     outlineSettings {
      bApplyOutlineToDropShadows=false
      bMiteredCorners=false
      bSeparateFillAlpha=false
      outlineMaterial=null
      outlineSize=0
      outlineColor=<rgba 0 0 0 1>
     }
    }
    shadowOffset=<0 0>
    strikeBrush {
     bIsDynamicallyLoaded=false
     drawAs=NoDrawType
     imageType=NoImage
     mirroring=NoMirror
     resourceName=None
     resourceObject=null
     tiling=NoTile
     imageSize=<32 32>
     margin=<ltrb 0 0 0 0>
     outlineSettings {
      bUseBrushTransparency=false
      roundingType=HalfHeightRadius
      width=0
      color {colorUseRule=UseColor_Specified, specifiedColor=<rgba 0 0 0 0>}
      cornerRadii=<0 0 0 0>
     }
     tintColor {colorUseRule=UseColor_Specified, specifiedColor=<rgba 1 1 1 1>}
     uVRegion {bIsValid=false, max=<0 0>, min=<0 0>}
    }
   }
   slot {class=SizeBoxSlot}
  }
  SettingName_Spacer_1 {
   widgetClassPath=&Spacer
   bIsVariable=true
   parent=VerticalBox_2
   widget=SettingName_Spacer_1
   properties {size=<1 10>}
   slot {class=VerticalBoxSlot}
  }
  HorizontalBox_3 {
   widgetClassPath=&HorizontalBox
   parent=VerticalBox_2
   widget=HorizontalBox_3
   slot {class=VerticalBoxSlot, properties {padding=<ltrb 10 0 0 0>}}
  }
  FortResponsivePanel_8 {
   widgetClassPath=&FortResponsivePanel
   parent=HorizontalBox_3
   widget=FortResponsivePanel_8
   slot {class=HorizontalBoxSlot, properties {verticalAlignment=VAlign_Center}}
  }
  JamKeyPreview_LoopPreview {
   widgetClassPath=&WBP_FMJamKeyPreview
   bIsVariable=true
   parent=FortResponsivePanel_8
   widget=JamKeyPreview_LoopPreview
   properties {bIsEnabled=false}
   slot {class=FortResponsivePanelSlot}
  }
  ModeSelector_LoopPreview {
   widgetClassPath=&WBP_FMJamModeSelector
   bIsVariable=true
   parent=HorizontalBox_3
   widget=ModeSelector_LoopPreview
   properties {bIsEnabled=false, hide Text Block=true}
   slot {class=HorizontalBoxSlot, properties {verticalAlignment=VAlign_Bottom, size {sizeRule=Fill, value=1}}}
  }
  UIKit_Divider_1 {
   widgetClassPath=&WBP_UIKit_Divider_Base
   bIsVariable=true
   parent=ConfirmSelection_VerticalBox
   widget=UIKit_Divider_1
   properties {bOverride_Cursor=true, div Width=2}
   slot {class=VerticalBoxSlot, properties {padding=<ltrb 0 15 0 0>, size {sizeRule=Fill, value=1}}}
  }
  FortResponsivePanel_4 {
   widgetClassPath=&FortResponsivePanel
   parent=ConfirmSelection_VerticalBox
   widget=FortResponsivePanel_4
   slot {class=VerticalBoxSlot, properties {padding=<ltrb 10 15 10 0>}}
  }
  ConfirmLoop_Button {
   widgetClassPath=&WBP_FMJamLoopButton
   bIsVariable=true
   parent=FortResponsivePanel_4
   widget=ConfirmLoop_Button
   properties {bOverride_Cursor=true, style=/FMJamSystem/Widgets/GlobalControls/Styles/WS_Jam_GC_Button_Yellow.WS_Jam_GC_Button_Yellow_C}
   slot {class=FortResponsivePanelSlot}
  }
  BackButtonSwitcher {
   widgetClassPath=&CommonVisibilitySwitcher
   bIsVariable=true
   parent=VerticalBox_223
   widget=BackButtonSwitcher
   slot {class=VerticalBoxSlot, properties {horizontalAlignment=HAlign_Center, verticalAlignment=VAlign_Center, padding=<ltrb 16 16 16 16>}}
  }
  Button_Close {
   widgetClassPath=&WBP_UIKit_Button_Regular
   bInherited=true
   bIsVariable=true
   parent=BackButtonSwitcher
   widget=Button_Close
   properties {minHeight=56, text=Back, triggeringInputAction {dataTable=&InputActionDataTable, rowName=BackToGame}}
   slot {class=CommonVisibilitySwitcherSlot}
  }
  Button_Back {
   widgetClassPath=&WBP_UIKit_Button_Regular
   bIsVariable=true
   parent=BackButtonSwitcher
   widget=Button_Back
   properties {minHeight=56, text=Cancel, triggeringInputAction {dataTable=&InputActionDataTable, rowName=BackToGame}}
   slot {class=CommonVisibilitySwitcherSlot}
  }
  Button_Cancel {
   widgetClassPath=&WBP_UIKit_Button_Regular
   bIsVariable=true
   parent=BackButtonSwitcher
   widget=Button_Cancel
   properties {minHeight=56, text=Cancel, triggeringInputAction {dataTable=&InputActionDataTable, rowName=BackToGame}}
   slot {class=CommonVisibilitySwitcherSlot}
  }
  Empty {widgetClassPath=&Overlay, parent=BackButtonSwitcher, widget=Empty, slot {class=CommonVisibilitySwitcherSlot}}
 ]
}
