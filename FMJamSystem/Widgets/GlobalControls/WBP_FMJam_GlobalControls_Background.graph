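# Backdrop widget for the FM Jam global controls, which signals that the controls should be hidden when it is clicked.
# An overlay stacks a semi-transparent black border, holding a background blur, beneath an invisible fill-sized generic button that catches clicks.
# On click, it finds the owning player's jam session and broadcasts a GlobalControls Displayed event set to false through that session's gameplay event router.
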WidgetBlueprint /FMJamSystem/Widgets/GlobalControls/WBP_FMJam_GlobalControls_Background
graphs [EventGraph {nodes [
   "On Clicked (WBP_UIKit_Button_Generic)"!() -> (OutputDelegate, then, Button:Common Button Base Object Reference) {is=title K2Node_ComponentBoundEvent}
   > n1 = GetJamSessionForActor!(execute, Actor:Actor Object Reference=GetOwningPlayer()) -> (then, ReturnValue:Jam Session Interface Interface)
   > Branch(Condition=IsValidInterface!(Interface:"Interface Interface (by ref)"=@n1) -> (ReturnValue:bool))
   n2 = GetGameplayEventRouterComponent!(self:Jam Session Interface Interface=@n1) -> (ReturnValue:Gameplay Event Router Component Object Reference)
   > Is Valid(InputObject=@n2)
   > Broadcast Event!(
    execute
    self:Gameplay Event Router Component Object Reference=@n2
    EventData:Jam Event Global Controls Displayed Structure=Make Jam Event Global Controls Displayed!(isDisplayed_1_#1:bool=False) -> (JamEvent_GlobalControls_Displayed:Jam Event Global Controls Displayed Structure) {is=title K2Node_MakeStruct}
    EventContext:Object Reference=GetOwningPlayer()
   ) -> (then, ReturnValue:bool)
   Event PreConstruct()
  ]}]
variables []
widgets {
 info {inheritedWidgetCount=0, parentClass=&UserWidget, rootWidgetClass=&Overlay, widgetCount=4}
 widgets [
  Overlay_0 {widgetClassPath=&Overlay, widget=Overlay_0}
  Border_Background {
   widgetClassPath=&Border
   bIsVariable=true
   parent=Overlay_0
   widget=Border_Background
   properties {
    horizontalAlignment=HAlign_Left
    verticalAlignment=VAlign_Bottom
    visibility=SelfHitTestInvisible
    background {
     bIsDynamicallyLoaded=false
     drawAs=Image
     imageType=NoImage
     mirroring=NoMirror
     resourceName=None
     resourceObject=null
     tiling=NoTile
     imageSize=<8589936640 0>
     margin=<ltrb 0 0 0 0>
     outlineSettings {
      bUseBrushTransparency=false
      roundingType=HalfHeightRadius
      width=0
      color {colorUseRule=UseColor_Specified, specifiedColor=<rgba 0 0 0 0>}
      cornerRadii=<0 0 0 0>
     }
     tintColor {colorUseRule=UseColor_Specified, specifiedColor=<rgba 0 0 0 0.6>}
     uVRegion {bIsValid=false, max=<0 0>, min=<0 0>}
    }
    padding=<ltrb 30 0 30 30>
   }
   slot {class=OverlaySlot, properties {horizontalAlignment=HAlign_Fill, verticalAlignment=VAlign_Fill}}
  }
  BackgroundBlur_36 {
   widgetClassPath=&BackgroundBlur
   parent=Border_Background
   widget=BackgroundBlur_36
   slot {class=BorderSlot, properties {horizontalAlignment=HAlign_Left, verticalAlignment=VAlign_Bottom, padding=<ltrb 30 0 30 30>}}
  }
  WBP_UIKit_Button_Generic {
   widgetClassPath=&WBP_UIKit_Button_Generic
   bIsVariable=true
   parent=Overlay_0
   widget=WBP_UIKit_Button_Generic
   properties {
    iconClass=null
    style=/FortUILibrary/Foundation/Buttons/ButtonStyles/ButtonStyle-Empty.ButtonStyle-Empty_C
    text=""
    colorAndOpacity=<rgba 1 1 1 0>
    font {
     bForceMonospaced=false
     bMaterialIsStencil=false
     fontMaterial=/FortUILibrary/UIKit/Buttons/MaterialInstances/CTAButtons/MI_UIKit_TextQuiet
     fontObject=&Burbank
     letterSpacing=48
     monospacedWidth=1
     size=21
     skewAmount=0
     typefaceFontName=BurbankRegular-Black
     outlineSettings {
      bApplyOutlineToDropShadows=false
      bMiteredCorners=false
      bSeparateFillAlpha=false
      outlineMaterial=null
      outlineSize=0
      outlineColor=<rgba 0 0 0 0>
     }
    }
    foregroundColor {colorUseRule=UseColor_Foreground, specifiedColor=<rgba 1 0 1 0>}
   }
   slot {class=OverlaySlot, properties {horizontalAlignment=HAlign_Fill, verticalAlignment=VAlign_Fill}}
  }
 ]
}
